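# Corrective Control Rig for the F_MED_RenegadeWhip character body, adjusting deform, belt, strap, jacket and extra foot bones.
# It runs shared DeformRig, then a sequence constraining extra foot/ball bones, adding foot offsets during emotes and shrinking upper-arm deform bones.
# Thigh rotation, pelvis height and spherical pose readers, some multiplied together, weight additional shoulder, belt, strap and jacket offsets.
# One belt-base adjustment is left disconnected from execution.
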ControlRigBlueprint /BRCosmetics/Characters/Player/Female/Medium/Bodies/F_MED_RenegadeWhip/Meshes/F_MED_RenegadeWhip_CtrlRig
graphs [
 RigVMFunctionLibrary {links [], nodes []}
 RigVMFunction_Sequence_1_ContainedGraph {nodes [
   n1 = RigVMFunctionEntryNode()
   RigVMFunction_Sequence = &RigVMFunction_Sequence(ExecuteContext=@n1.ExecuteContext)
   n2 = &RigVMFunction_Sequence(ExecuteContext=@RigVMFunction_Sequence.B)
   n3 = &RigVMFunction_Sequence(ExecuteContext=@n2.B)
   n4 = &RigVMFunction_Sequence(ExecuteContext=@n3.B)
   n5 = &RigVMFunction_Sequence(ExecuteContext=@n4.B)
   n6 = &RigVMFunction_Sequence(ExecuteContext=@n5.B)
   n7 = &RigVMFunction_Sequence(ExecuteContext=@n6.B)
   RigVMFunctionReturnNode(A=@RigVMFunction_Sequence.A, B=@n2.A, C=@n3.A, D=@n4.A, E=@n5.A, F=@n6.A, G=@n7.A, H=@n7.B)
  ]}
 RigVMFunction_Sequence_4_ContainedGraph {nodes [
   n1 = RigVMFunctionEntryNode()
   n2 = &RigVMFunction_Sequence(ExecuteContext=@n1.ExecuteContext)
   n3 = &RigVMFunction_Sequence(ExecuteContext=@n2.B)
   RigVMFunctionReturnNode(B=@n3.A, C=@n3.B, A=@n2.A)
  ]}
 RigVMModel {nodes [
   RigUnit_BeginExecution = &RigUnit_BeginExecution()
   FN_DeformRig_FNC = RigVMFunctionReferenceNode(
    DeformRig_ Curve="(deformRig_l_upperarm_up_alpha_72_#1=False,deformRig_r_upperarm_up_alpha_73_#2=False,deformRig_l_upperarm_fr_alpha_74_#3=False,deformRig_r_upperarm_fr_alpha_75_CB51B8 ...(truncated)"
    DeformRig_Bool="(upperarm_1_#4=True,elbow_3_#5=True,wrist_7_#6=True,groin_8_#7=True,glute_17_#8=True,knee_9_40C408324CA261 ...(truncated)"
    DeformRig_Debug=(upperarm_debug_18_#4=False,wrist_debug_20_#6=False,lat_debug_26_#9=False,groinGlute_debug_27_#7=False)
    DeformRig_Value="(upperarm_up_18_#10=(Rotation=(X=0.0,Y=-0.0,Z=0.0,W=1.0),Translation=(X=0.0,Y=0.0,Z=8.0),Scale3D=(X=0.70,Y=1.0,Z=1.0)),upperarm_fwd_fr_19_#11=(Rotation=(X=0.0,Y=-0.0,Z=0.0,W=1.0),Translation=(X=0.0, ...(truncated)"
    ExecuteContext=@RigUnit_BeginExecution.ExecutePin
   )
   > RigVMFunction_Sequence = RigVMAggregateNode()
   Get Delta Ref_3 = RigVMFunctionReferenceNode(Item=(Type=Bone,Name="foot_l"), ExecuteContext=@RigVMFunction_Sequence.A)
   > Get Delta Ref_4 = RigVMFunctionReferenceNode(Item=(Type=Bone,Name="foot_r"))
   > &RigUnit_RotationConstraintLocalSpaceOffset(
    AdvancedSettings=(InterpolationType=Average,RotationOrderForFilter=XZY)
    Child=(Type=Bone,Name="extra_foot_r")
    Filter=(bX=True,bY=True,bZ=True)
    Parents=((Item=(Type=Bone,Name="foot_r"),Weight=1))
    Weight=1
    bMaintainOffset=true
   )
   > &RigUnit_RotationConstraintLocalSpaceOffset(
    AdvancedSettings=(InterpolationType=Average,RotationOrderForFilter=XZY)
    Child=(Type=Bone,Name="extra_ball_r")
    Filter=(bX=True,bY=True,bZ=True)
    Parents=((Item=(Type=Bone,Name="ball_r"),Weight=1))
    Weight=1
    bMaintainOffset=true
   )
   > &RigUnit_RotationConstraintLocalSpaceOffset(
    AdvancedSettings=(InterpolationType=Average,RotationOrderForFilter=XZY)
    Child=(Type=Bone,Name="extra_foot_l")
    Filter=(bX=True,bY=True,bZ=True)
    Parents=((Item=(Type=Bone,Name="foot_l"),Weight=1))
    Weight=1
    bMaintainOffset=true
   )
   > n1 = &RigUnit_RotationConstraintLocalSpaceOffset(
    AdvancedSettings=(InterpolationType=Average,RotationOrderForFilter=XZY)
    Child=(Type=Bone,Name="extra_ball_l")
    Filter=(bX=True,bY=True,bZ=True)
    Parents=((Item=(Type=Bone,Name="ball_l"),Weight=1))
    Weight=1
    bMaintainOffset=true
   )
   RigVMFunction_ControlFlowBranch = &RigVMFunction_ControlFlowBranch(ExecuteContext=@n1.ExecutePin, Condition=RigVMVariableNode() {variable=IsPlayEmote}.Value)
   &RigUnit_ModifyTransforms(
    ItemToModify=((Item=(Type=Bone,Name="extra_foot_l"),Transform=(Rotation=(X=0,Y=0,Z=-0.1305,W=0.9914),Translation=(X=1,Y=-6,Z=0),Scale3D=(X=1,Y=1,Z=1))))
    WeightMinimum=0
    Weight=&RigVMFunction_MathDoubleRemap(SourceMaximum=60, SourceMinimum=0, TargetMaximum=1, TargetMinimum=0, Value=@"Get Delta Ref_3".Result.Rotation.Yaw).Result
    ExecutePin=@RigVMFunction_ControlFlowBranch.True
   )
   > &RigUnit_ModifyTransforms(
    ItemToModify=((Item=(Type=Bone,Name="extra_foot_r"),Transform=(Rotation=(X=0,Y=0,Z=-0.1305,W=0.9914),Translation=(X=1,Y=-6,Z=0),Scale3D=(X=1,Y=1,Z=1))))
    WeightMinimum=0
    Weight=&RigVMFunction_MathDoubleRemap(SourceMaximum=60, SourceMinimum=0, TargetMaximum=1, TargetMinimum=0, Value=@"Get Delta Ref_4".Result.Rotation.Yaw).Result
   )
   &RigUnit_ModifyTransforms(
    ItemToModify=((Item=(Type=Bone,Name="deform_upperarm_l"),Transform=(Rotation=(X=0,Y=0,Z=0,W=1),Translation=(X=0,Y=0,Z=0),Scale3D=(X=0.01,Y=0.01,Z=0.01))))
    Weight=1
    WeightMinimum=0
    ExecutePin=@RigVMFunction_Sequence.F
   )
   > &RigUnit_ModifyTransforms(
    ItemToModify=((Item=(Type=Bone,Name="deform_upperarm_r"),Transform=(Rotation=(X=0,Y=0,Z=0,W=1),Translation=(X=0,Y=0,Z=0),Scale3D=(X=0.01,Y=0.01,Z=0.01))))
    Weight=1
    WeightMinimum=0
   )
   n2 = &RigUnit_SphericalPoseReader(
    ActiveRegionScaleFactors=(PositiveWidth=1,NegativeWidth=1,PositiveHeight=1,NegativeHeight=1)
    ActiveRegionSize=0.02
    Debug=(bDrawDebug=true,bDraw2D=False,bDrawLocalAxes=False,DebugScale=25,DebugSegments=20,DebugThickness=0.25)
    DriverAxis=(X=-1,Y=0,Z=0)
    DriverItem=(Type=Bone,Name="upperarm_r")
    FalloffRegionScaleFactors=(PositiveWidth=1,NegativeWidth=1,PositiveHeight=1,NegativeHeight=1)
    FalloffSize=0.2
    FlipHeightScaling=False
    FlipWidthScaling=False
    OptionalParentItem=(Type=Bone,Name="None")
    RotationOffset=(X=-115,Y=0,Z=40)
    ExecutePin=@RigVMFunction_Sequence.G
   )
   > n3 = &RigUnit_SphericalPoseReader(
    ActiveRegionScaleFactors=(PositiveWidth=1,NegativeWidth=1,PositiveHeight=1,NegativeHeight=1)
    ActiveRegionSize=0.02
    Debug=(bDrawDebug=true,bDraw2D=False,bDrawLocalAxes=False,DebugScale=25,DebugSegments=20,DebugThickness=0.25)
    DriverAxis=(X=0,Y=0,Z=-1)
    DriverItem=(Type=Bone,Name="upperarm_r")
    FalloffRegionScaleFactors=(PositiveWidth=1,NegativeWidth=1,PositiveHeight=1,NegativeHeight=1)
    FalloffSize=0.2
    FlipHeightScaling=False
    FlipWidthScaling=False
    OptionalParentItem=(Type=Bone,Name="None")
    RotationOffset=(X=0,Y=135,Z=10)
   )
   > &RigUnit_ModifyTransforms(
    ItemToModify=((Item=(Type=Bone,Name="deform_upperarm_bk_r"),Transform=(Rotation=(X=0,Y=0,Z=0,W=1),Translation=(X=-7,Y=5,Z=0),Scale3D=(X=1,Y=1,Z=1))))
    WeightMinimum=0
    Weight=&RigVMFunction_MathDoubleMul(B=@n2.OutputParam, A=@n3.OutputParam).Result
   )
   &RigUnit_ModifyTransforms(
    ItemToModify=((Item=(Type=Bone,Name="dyn_strap_base"),Transform=(Rotation=(X=0,Y=0,Z=0,W=1),Translation=(X=0,Y=0,Z=0.5),Scale3D=(X=1,Y=1,Z=1))))
    Weight=1
    WeightMinimum=0
    ExecutePin=@RigVMFunction_Sequence.D
   )
   > Get Delta Ref = RigVMFunctionReferenceNode(Item=(Type=Bone,Name="thigh_l"))
   Remap = &RigVMFunction_MathDoubleRemap(SourceMaximum=-75, SourceMinimum=-20, TargetMaximum=1, TargetMinimum=0, Value=@"Get Delta Ref".Result.Rotation.Yaw)
   AlphaInterp = &RigVMFunction_AlphaInterp(InterpSpeedDecreasing=10, InterpSpeedIncreasing=50, Scale=1, bClampResult=true, bMapRange=false, Value=@Remap.Result)
   > &RigUnit_ModifyTransforms(
    ItemToModify=((Item=(Type=Bone,Name="deform_belt_base"),Transform=(Rotation=(X=0.0023,Y=0.0173,Z=-0.1305,W=0.9913),Translation=(X=-3,Y=-1,Z=5.5),Scale3D=(X=1,Y=1,Z=0.95))))
    WeightMinimum=0
    Weight=@AlphaInterp.Result
   )
   > &RigUnit_ModifyTransforms(
    ItemToModify=((Item=(Type=Bone,Name="dyn_strap_base"),Transform=(Rotation=(X=-0.13,Y=0.0864,Z=0.0114,W=0.9877),Translation=(X=0.4,Y=0.1,Z=2),Scale3D=(X=1,Y=1,Z=1))))
    WeightMinimum=0
    Weight=@AlphaInterp.Result
   )
   > &RigUnit_ModifyTransforms(
    ItemToModify=((Item=(Type=Bone,Name="deform_belt_side_l"),Transform=(Rotation=(X=0,Y=-0.1736,Z=0,W=0.9848),Translation=(X=-21,Y=0,Z=-1),Scale3D=(X=1,Y=1,Z=1))))
    WeightMinimum=0
    Weight=@AlphaInterp.Result
   )
   > n4 = &RigUnit_SphericalPoseReader(
    ActiveRegionScaleFactors=(PositiveWidth=1,NegativeWidth=1,PositiveHeight=1,NegativeHeight=1)
    ActiveRegionSize=0.025
    Debug=(bDrawDebug=false,bDraw2D=False,bDrawLocalAxes=False,DebugScale=25,DebugSegments=20,DebugThickness=0.25)
    DriverAxis=(X=-1,Y=0,Z=0)
    DriverItem=(Type=Bone,Name="thigh_l")
    FalloffRegionScaleFactors=(PositiveWidth=1,NegativeWidth=1,PositiveHeight=1,NegativeHeight=1)
    FalloffSize=0.15
    FlipHeightScaling=False
    FlipWidthScaling=False
    OptionalParentItem=(Type=Bone,Name="None")
    RotationOffset=(X=75,Y=0,Z=5)
   )
   > &RigUnit_ModifyTransforms(
    ItemToModify=((Item=(Type=Bone,Name="deform_belt_base"),Transform=(Rotation=(X=-0.1372,Y=-0.008,Z=-0.2619,W=0.9553),Translation=(X=1,Y=0,Z=0),Scale3D=(X=1,Y=1,Z=1))))
    WeightMinimum=0
    Weight=@n4.OutputParam
   )
   > &RigUnit_ModifyTransforms(
    ItemToModify=((Item=(Type=Bone,Name="dyn_strap_base"),Transform=(Rotation=(X=0,Y=0,Z=0,W=1),Translation=(X=0.2,Y=0.2,Z=0),Scale3D=(X=1,Y=1,Z=1))))
    WeightMinimum=0
    Weight=@n4.OutputParam
   )
   > &RigUnit_ModifyTransforms(
    ItemToModify=((Item=(Type=Bone,Name="deform_belt_01"),Transform=(Rotation=(X=0,Y=0,Z=0,W=1),Translation=(X=0,Y=0,Z=0),Scale3D=(X=1,Y=1,Z=1))))
    WeightMinimum=0
    Weight=@n4.OutputParam
   )
   > n5 = &RigUnit_SphericalPoseReader(
    ActiveRegionScaleFactors=(PositiveWidth=1,NegativeWidth=1,PositiveHeight=1,NegativeHeight=1)
    ActiveRegionSize=0.01
    Debug=(bDrawDebug=false,bDraw2D=False,bDrawLocalAxes=False,DebugScale=25,DebugSegments=20,DebugThickness=0.25)
    DriverAxis=(X=-1,Y=0,Z=0)
    DriverItem=(Type=Bone,Name="thigh_l")
    FalloffRegionScaleFactors=(PositiveWidth=1,NegativeWidth=1,PositiveHeight=1,NegativeHeight=1)
    FalloffSize=0.1
    FlipHeightScaling=False
    FlipWidthScaling=False
    OptionalParentItem=(Type=Bone,Name="None")
    RotationOffset=(X=85,Y=0,Z=-5)
   )
   > n6 = &RigUnit_SphericalPoseReader(
    ActiveRegionScaleFactors=(PositiveWidth=1,NegativeWidth=1,PositiveHeight=1,NegativeHeight=1)
    ActiveRegionSize=0.01
    Debug=(bDrawDebug=false,bDraw2D=False,bDrawLocalAxes=False,DebugScale=25,DebugSegments=20,DebugThickness=0.25)
    DriverAxis=(X=0,Y=0,Z=-1)
    DriverItem=(Type=Bone,Name="thigh_l")
    FalloffRegionScaleFactors=(PositiveWidth=1,NegativeWidth=1,PositiveHeight=1,NegativeHeight=1)
    FalloffSize=0.1
    FlipHeightScaling=False
    FlipWidthScaling=False
    OptionalParentItem=(Type=Bone,Name="None")
    RotationOffset=(X=5,Y=155,Z=0)
   )
   Multiply = &RigVMFunction_MathDoubleMul(A=@n5.OutputParam, B=@n6.OutputParam)
   > &RigUnit_ModifyTransforms(
    ItemToModify=((Item=(Type=Bone,Name="deform_belt_base"),Transform=(Rotation=(X=0.0153,Y=-0.0961,Z=-0.2957,W=0.9503),Translation=(X=-1,Y=-2,Z=0),Scale3D=(X=1,Y=1,Z=1))))
    WeightMinimum=0
    Weight=@Multiply.Result
   )
   RigVMFunction_Sequence_3 = RigVMAggregateNode(ExecuteContext=@RigVMFunction_Sequence.E)
   Get Delta Ref_1 = RigVMFunctionReferenceNode(Item=(Type=Bone,Name="thigh_l"), ExecuteContext=@RigVMFunction_Sequence_3.A)
   > &RigUnit_ModifyTransforms(
    ItemToModify=((Item=(Type=Bone,Name="dyn_jacket_base_fr_l"),Transform=(Rotation=(X=0.2164,Y=0,Z=0,W=0.9763),Translation=(X=-2,Y=0,Z=0),Scale3D=(X=1,Y=1,Z=1))))
    WeightMinimum=0
    Weight=&RigVMFunction_MathDoubleRemap(SourceMaximum=-90, SourceMinimum=-20, TargetMaximum=1, TargetMinimum=0, Value=@"Get Delta Ref_1".Result.Rotation.Yaw).Result
   )
   > n7 = &RigUnit_SphericalPoseReader(
    ActiveRegionScaleFactors=(PositiveWidth=1,NegativeWidth=1,PositiveHeight=1,NegativeHeight=1)
    ActiveRegionSize=0.025
    Debug=(bDrawDebug=true,bDraw2D=False,bDrawLocalAxes=False,DebugScale=25,DebugSegments=20,DebugThickness=0.25)
    DriverAxis=(X=0,Y=1,Z=0)
    DriverItem=(Type=Bone,Name="spine_03")
    FalloffRegionScaleFactors=(PositiveWidth=1,NegativeWidth=1,PositiveHeight=1,NegativeHeight=1)
    FalloffSize=0.2
    FlipHeightScaling=False
    FlipWidthScaling=False
    OptionalParentItem=(Type=Bone,Name="pelvis")
    RotationOffset=(X=55,Y=25,Z=0)
   )
   > n8 = &RigUnit_SphericalPoseReader(
    ActiveRegionScaleFactors=(PositiveWidth=1,NegativeWidth=1,PositiveHeight=1,NegativeHeight=1)
    ActiveRegionSize=0.05
    Debug=(bDrawDebug=true,bDraw2D=False,bDrawLocalAxes=False,DebugScale=25,DebugSegments=20,DebugThickness=0.25)
    DriverAxis=(X=-1,Y=0,Z=0)
    DriverItem=(Type=Bone,Name="thigh_l")
    FalloffRegionScaleFactors=(PositiveWidth=1,NegativeWidth=1,PositiveHeight=1,NegativeHeight=1)
    FalloffSize=0.2
    FlipHeightScaling=False
    FlipWidthScaling=False
    OptionalParentItem=(Type=Bone,Name="None")
    RotationOffset=(X=120,Y=0,Z=-10)
   )
   n9 = &RigVMFunction_MathDoubleMul(A=@n7.OutputParam, B=@n8.OutputParam)
   > &RigUnit_ModifyTransforms(
    ItemToModify=((Item=(Type=Bone,Name="dyn_jacket_base_fr_l"),Transform=(Rotation=(X=0.4707,Y=0,Z=0,W=0.8823),Translation=(X=-0.5,Y=-2,Z=0),Scale3D=(X=1,Y=1,Z=1))))
    WeightMinimum=0
    Weight=@n9.Result
   )
   > &RigUnit_ModifyTransforms(
    ItemToModify=((Item=(Type=Bone,Name="dyn_jacket_base_side_l"),Transform=(Rotation=(X=0.2164,Y=0,Z=0,W=0.9763),Translation=(X=-1,Y=5,Z=0),Scale3D=(X=1,Y=1,Z=1))))
    WeightMinimum=0
    Weight=@n9.Result
   )
   Get Delta Ref_2 = RigVMFunctionReferenceNode(Item=(Type=Bone,Name="thigh_r"), ExecuteContext=@RigVMFunction_Sequence_3.B)
   > &RigUnit_ModifyTransforms(
    ItemToModify=((Item=(Type=Bone,Name="dyn_jacket_base_fr_r"),Transform=(Rotation=(X=0.0038,Y=-0.0435,Z=0.0871,W=0.9952),Translation=(X=-6,Y=0,Z=0),Scale3D=(X=1,Y=1,Z=1))))
    WeightMinimum=0
    Weight=&RigVMFunction_MathDoubleRemap(SourceMaximum=-110, SourceMinimum=-75, TargetMaximum=1, TargetMinimum=0, Value=@"Get Delta Ref_2".Result.Rotation.Yaw).Result
   )
   n10 = &RigVMFunction_MathDoubleRemap(
    SourceMaximum=12
    SourceMinimum=30
    TargetMaximum=1
    TargetMinimum=0
    Value=&RigUnit_GetRelativeTransformForItem(Child=(Type=Bone,Name="pelvis"), Parent=(Type=Bone,Name="root"), bChildInitial=False, bParentInitial=False).RelativeTransform.Translation.Z
   )
   &RigUnit_ModifyTransforms(
    ItemToModify=((Item=(Type=Bone,Name="dyn_jacket_base_fr_r"),Transform=(Rotation=(X=0,Y=0,Z=0,W=1),Translation=(X=-10,Y=0,Z=0),Scale3D=(X=1,Y=1,Z=1))))
    WeightMinimum=0
    Weight=@n10.Result
    ExecutePin=@RigVMFunction_Sequence_3.C
   )
   > &RigUnit_ModifyTransforms(
    ItemToModify=((Item=(Type=Bone,Name="dyn_jacket_base_side_r"),Transform=(Rotation=(X=0,Y=0,Z=0,W=1),Translation=(X=-6,Y=0,Z=0),Scale3D=(X=1,Y=1,Z=1))))
    WeightMinimum=0
    Weight=@n10.Result
   )
   > &RigUnit_ModifyTransforms(
    ItemToModify=((Item=(Type=Bone,Name="dyn_jacket_base_bk_r"),Transform=(Rotation=(X=0,Y=0,Z=0,W=1),Translation=(X=-6,Y=0,Z=0),Scale3D=(X=1,Y=1,Z=1))))
    WeightMinimum=0
    Weight=@n10.Result
   )
   > &RigUnit_ModifyTransforms(
    ItemToModify=((Item=(Type=Bone,Name="dyn_jacket_base_fr_l"),Transform=(Rotation=(X=0,Y=0,Z=0,W=1),Translation=(X=-1,Y=0,Z=0),Scale3D=(X=1,Y=1,Z=1))))
    WeightMinimum=0
    Weight=@n10.Result
   )
   > &RigUnit_ModifyTransforms(
    ItemToModify=((Item=(Type=Bone,Name="dyn_jacket_base_side_l"),Transform=(Rotation=(X=0,Y=0,Z=0,W=1),Translation=(X=-5,Y=0,Z=0),Scale3D=(X=1,Y=1,Z=1))))
    WeightMinimum=0
    Weight=@n10.Result
   )
   > &RigUnit_ModifyTransforms(
    ItemToModify=((Item=(Type=Bone,Name="dyn_jacket_base_bk_l"),Transform=(Rotation=(X=0,Y=0,Z=0,W=1),Translation=(X=-3,Y=-5,Z=0),Scale3D=(X=1,Y=1,Z=1))))
    WeightMinimum=0
    Weight=@n10.Result
   )
   > &RigUnit_ModifyTransforms(
    ItemToModify=((Item=(Type=Bone,Name="dyn_jacket_base_mid"),Transform=(Rotation=(X=0,Y=0,Z=0,W=1),Translation=(X=-5,Y=0,Z=0),Scale3D=(X=1,Y=1,Z=1))))
    WeightMinimum=0
    Weight=@n10.Result
   )
   &RigUnit_ModifyTransforms(
    ExecutePin=()
    ItemToModify=((Item=(Type=Bone,Name="deform_belt_base"),Transform=(Rotation=(X=-0.1115,Y=0.1397,Z=-0.5929,W=0.7852),Translation=(X=-9,Y=-1,Z=0),Scale3D=(X=1,Y=1,Z=1))))
    WeightMinimum=0
    Weight=@Remap.Result
   )
  ]}
]
hierarchy [
 root {loc=<0 0 0>, rot=<0 0 0 1>}
 pelvis {parent=root, loc=<0 -0.5301 98.0571>, rot=<0.0127 0.707 -0.0127 -0.707>}
 spine_01 {parent=pelvis, loc=<6.1208 -0.9328 0>, rot=<0 0 0.034 -0.9994>}
 spine_02 {parent=spine_01, loc=<6.5996 -0.2095 0>, rot=<0 0 0.0238 -0.9997>}
 spine_03 {parent=spine_02, loc=<6.4914 -0.1237 0>, rot=<0 0 0.0149 -0.9999>}
 spine_04 {parent=spine_03, loc=<6.9598 0.0794 0>, rot=<0 0 -0.0132 -0.9999>}
 spine_05 {parent=spine_04, loc=<6.9987 0 0>, rot=<0 0 -0.0015 -1>}
 attach_cape {parent=spine_05, loc=<9.0212 -11.7149 -0.0003>, rot=<0.0616 -0.7044 -0.0616 -0.7044>}
 neck_01 {parent=spine_05, loc=<11.3138 0.1529 0>, rot=<0 0 -0.2003 -0.9797>}
 neck_02 {parent=neck_01, loc=<4.5627 0 0>, rot=<0 0 0.0504 -0.9987>}
 head {parent=neck_02, loc=<5.3914 0 0>, rot=<0 0 0.115 -0.9934>}
 attach_backpack {parent=spine_05, loc=<-0.1175 -13.4739 -0.0004>, rot=<-0.0093 -0.707 0.0093 -0.707>}
 clavicle_l {parent=spine_05, loc=<5.4831 -0.1053 -3.2946>, rot=<0.0465 -0.7289 -0.009 -0.6829>}
 upperarm_l {parent=clavicle_l, loc=<9.8658 -1.2136 -0.4105>, rot=<0.0164 -0.3245 -0.0269 -0.9454>}
 lowerarm_l {parent=upperarm_l, loc=<26.465 -0.8345 0.259>, rot=<0.0036 0.0161 -0.2298 -0.9731>}
 hand_l {parent=lowerarm_l, loc=<22.7355 -0.3203 -0.6105>, rot=<-0.6784 -0.0223 0.0735 -0.7307>}
 thumb_01_l {parent=hand_l, loc=<1.3118 -0.3191 -2.459>, rot=<0.6546 -0.4042 0.0664 -0.6354>}
 thumb_02_l {parent=thumb_01_l, loc=<5.9336 -0.0448 0.0401>, rot=<-0.0037 0.0213 0.298 -0.9543>}
 thumb_03_l {parent=thumb_02_l, loc=<2.2223 0.3682 -0.0907>, rot=<0.0118 -0.013 0.0066 -0.9998>}
 index_metacarpal_l {parent=hand_l, loc=<1.6535 0.0183 -2.96>, rot=<0.0158 -0.0124 -0.0173 -0.9996>}
 index_01_l {parent=index_metacarpal_l, loc=<8.3008 0.0357 -0.0524>, rot=<-0.0044 0.0485 0.0924 -0.9945>}
 index_02_l {parent=index_01_l, loc=<3.3587 -0.3004 0.0452>, rot=<-0.002 0.0027 0.3225 -0.9466>}
 index_03_l {parent=index_02_l, loc=<2.5578 -0.123 -0.0252>, rot=<-0.002 0.0067 0.2558 -0.9667>}
 middle_metacarpal_l {parent=hand_l, loc=<2.1856 -0.1725 -0.5984>, rot=<0.006 0.0168 -0.0329 -0.9993>}
 middle_01_l {parent=middle_metacarpal_l, loc=<7.3275 -0.2447 0.0129>, rot=<-0.0208 0.0042 0.0858 -0.9961>}
 middle_02_l {parent=middle_01_l, loc=<4.0123 0.1401 0.1175>, rot=<-0.0218 0.0258 0.3512 -0.9357>}
 middle_03_l {parent=middle_02_l, loc=<2.6077 0.0534 0.1373>, rot=<0.0082 0.0188 0.1708 -0.9851>}
 ring_metacarpal_l {parent=hand_l, loc=<2.2074 -0.3625 1.3638>, rot=<-0.0284 0.0722 -0.0458 -0.9959>}
 ring_01_l {parent=ring_metacarpal_l, loc=<7.0295 -0.6197 -0.1446>, rot=<-0.0489 -0.0252 0.1892 -0.9804>}
 ring_02_l {parent=ring_01_l, loc=<3.6174 0.4647 -0.1027>, rot=<0 0 0.2806 -0.9598>}
 ring_03_l {parent=ring_02_l, loc=<2.4152 0.114 0.0199>, rot=<0 0 0.1189 -0.9929>}
 pinky_metacarpal_l {parent=hand_l, loc=<1.7616 -0.9753 3.3814>, rot=<-0.0397 0.0814 -0.069 -0.9935>}
 pinky_01_l {parent=pinky_metacarpal_l, loc=<7.1709 -0.4831 -0.1094>, rot=<-0.0654 -0.0035 0.2242 -0.9723>}
 pinky_02_l {parent=pinky_01_l, loc=<2.8274 0.0893 -0.1043>, rot=<0.0037 0.0011 0.2854 -0.9584>}
 pinky_03_l {parent=pinky_02_l, loc=<2.2046 0.0307 -0.0078>, rot=<-0.0002 0.0098 0.1445 -0.9895>}
 deform_wrist_dn_base_l {parent=hand_l, loc=<0 -1.5326 0>, rot=<0 0 0 -1>}
 deform_wrist_dn_l {parent=deform_wrist_dn_base_l, loc=<0 0 0>, rot=<0 0 0 1>}
 deform_wrist_up_base_l {parent=hand_l, loc=<0 2.3009 0>, rot=<0 0 0 -1>}
 deform_wrist_up_l {parent=deform_wrist_up_base_l, loc=<0 0 0>, rot=<0 0 0 1>}
 extra_thumb_01_l {parent=hand_l, loc=<1.4323 -0.4697 -2.4811>, rot=<0.559 -0.5964 -0.0644 -0.5724>}
 extra_thumb_02_l {parent=extra_thumb_01_l, loc=<4.334 -0.6908 -0.0535>, rot=<0.0283 -0.0467 -0.0934 -0.9941>}
 extra_thumb_03_l {parent=extra_thumb_02_l, loc=<2.3506 -0.8142 -0.0135>, rot=<-0.0443 0.0172 0.1846 -0.9817>}
 extra_middle_metacarpal_l {parent=hand_l, loc=<2.159 1.0686 -0.813>, rot=<0.0057 0.0178 -0.0432 -0.9989>}
 extra_middle_01_l {parent=extra_middle_metacarpal_l, loc=<7.0585 -0.7053 -0.0467>, rot=<-0.0439 0.0303 -0.0494 -0.9974>}
 extra_middle_02_l {parent=extra_middle_01_l, loc=<3.2235 0.0147 0.1714>, rot=<0.0123 0.0037 0.2499 -0.9682>}
 extra_middle_03_l {parent=extra_middle_02_l, loc=<2.4969 -0.5452 0.2193>, rot=<-0.029 0.0239 0.2199 -0.9748>}
 extra_ring_metacarpal_l {parent=hand_l, loc=<2.0979 0.8982 1.2513>, rot=<-0.0276 0.1212 -0.0671 -0.99>}
 extra_ring_01_l {parent=extra_ring_metacarpal_l, loc=<6.7958 -1.0874 -0.3243>, rot=<-0.0404 0.0336 0.0192 -0.9984>}
 extra_ring_02_l {parent=extra_ring_01_l, loc=<3.0697 -0.3576 -0.0417>, rot=<0.0122 0.0002 0.1863 -0.9824>}
 extra_ring_03_l {parent=extra_ring_02_l, loc=<2.2631 -0.522 -0.0949>, rot=<-0.0393 0.0033 0.1833 -0.9823>}
 extra_pinky_metacarpal_l {parent=hand_l, loc=<1.9652 0.3365 2.8772>, rot=<-0.0568 0.2106 -0.0463 -0.9748>}
 extra_pinky_01_l {parent=extra_pinky_metacarpal_l, loc=<6.7076 -0.1771 -0.2656>, rot=<-0.0384 0.0277 -0.0662 -0.9967>}
 extra_pinky_02_l {parent=extra_pinky_01_l, loc=<2.2376 -0.139 0.0637>, rot=<0.0092 0.0197 0.1204 -0.9925>}
 extra_pinky_03_l {parent=extra_pinky_02_l, loc=<2.0115 -0.9666 -0.1014>, rot=<-0.0143 -0.0097 0.2565 -0.9664>}
 extra_index_metacarpal_l {parent=hand_l, loc=<2.0415 0.9574 -2.9502>, rot=<0.0156 -0.0287 -0.0299 -0.999>}
 extra_index_01_l {parent=extra_index_metacarpal_l, loc=<7.5923 -0.4742 -0.0494>, rot=<-0.0343 -0.0166 0.0601 -0.9975>}
 extra_index_02_l {parent=extra_index_01_l, loc=<2.6596 -0.3654 -0.0719>, rot=<0.0282 -0.003 0.1005 -0.9945>}
 extra_index_03_l {parent=extra_index_02_l, loc=<2.4177 -1.0107 -0.0693>, rot=<0.0055 0.0046 0.2187 -0.9758>}
 hand_attach_l {parent=hand_l, loc=<8.5 -3 0>, rot=<0.7071 0.7071 0 0>}
 weapon_l {parent=hand_l, loc=<-0.1974 -3.4111 0.3811>, rot=<0 0 0 1>}
 lowerarm_twist_02_l {parent=lowerarm_l, loc=<10.6707 0 0>, rot=<0 0 0 1>}
 lowerarm_twist_01_l {parent=lowerarm_l, loc=<18.316 0 0>, rot=<0 0 0 1>}
 deform_elbow_in_base_l {parent=lowerarm_l, loc=<0.1834 3.6062 -0.2572>, rot=<0 0 0 -1>}
 deform_elbow_in_l {parent=deform_elbow_in_base_l, loc=<0 0 0>, rot=<0 0 0 1>}
 deform_elbow_out_base_l {parent=lowerarm_l, loc=<0 -4 0>, rot=<0 0 0 1>}
 deform_elbow_out_l {parent=deform_elbow_out_base_l, loc=<0 0 0>, rot=<0 0 0 1>}
 upperarm_twist_01_l {parent=upperarm_l, loc=<7.2135 -0.1134 0>, rot=<0 0 0 1>}
 deform_upperarm_bk_base_l {parent=upperarm_twist_01_l, loc=<-6.4053 -4.3746 0.314>, rot=<0 0 0 -1>}
 deform_upperarm_bk_l {parent=deform_upperarm_bk_base_l, loc=<0 0 0>, rot=<0 0 0 1>}
 deform_upperarm_fr_base_l {parent=upperarm_twist_01_l, loc=<-5.6312 5.092 0.3354>, rot=<0 0 0 -1>}
 deform_upperarm_fr_l {parent=deform_upperarm_fr_base_l, loc=<0 0 0>, rot=<0 0 0 1>}
 deform_upperarm_up_base_l {parent=upperarm_twist_01_l, loc=<-8.891 0.4825 3.9517>, rot=<0 0 0 -1>}
 deform_upperarm_up_l {parent=deform_upperarm_up_base_l, loc=<0 0 0>, rot=<0 0 0 1>}
 upperarm_twist_02_l {parent=upperarm_l, loc=<19.7998 -0.1134 0>, rot=<0 0 0 1>}
 deform_upperarm_l {parent=upperarm_l, loc=<8.342 -0.598 -0.7828>, rot=<-0.0065 -0.6629 -0.004 -0.7486>}
 clavicle_r {parent=spine_05, loc=<5.4831 -0.1053 3.2946>, rot=<0.7289 0.0465 -0.6829 0.009>}
 upperarm_r {parent=clavicle_r, loc=<-9.8658 1.2136 0.4105>, rot=<0.0164 -0.3245 -0.0269 -0.9454>}
 lowerarm_r {parent=upperarm_r, loc=<-26.465 0.8345 -0.259>, rot=<0.0036 0.0161 -0.2298 -0.9731>}
 hand_r {parent=lowerarm_r, loc=<-22.7355 0.3203 0.6106>, rot=<-0.6784 -0.0223 0.0735 -0.7307>}
 thumb_01_r {parent=hand_r, loc=<-1.3118 0.3191 2.459>, rot=<0.6546 -0.4042 0.0664 -0.6354>}
 thumb_02_r {parent=thumb_01_r, loc=<-5.9336 0.0448 -0.0401>, rot=<-0.0037 0.0213 0.298 -0.9543>}
 thumb_03_r {parent=thumb_02_r, loc=<-2.2223 -0.3682 0.0907>, rot=<0.0118 -0.013 0.0066 -0.9998>}
 index_metacarpal_r {parent=hand_r, loc=<-1.6535 -0.0183 2.96>, rot=<0.0158 -0.0124 -0.0173 -0.9996>}
 index_01_r {parent=index_metacarpal_r, loc=<-8.3008 -0.0357 0.0524>, rot=<-0.0044 0.0485 0.0924 -0.9945>}
 index_02_r {parent=index_01_r, loc=<-3.3587 0.3004 -0.0452>, rot=<-0.002 0.0027 0.3225 -0.9466>}
 index_03_r {parent=index_02_r, loc=<-2.5578 0.123 0.0252>, rot=<-0.002 0.0067 0.2558 -0.9667>}
 middle_metacarpal_r {parent=hand_r, loc=<-2.1856 0.1725 0.5984>, rot=<0.006 0.0168 -0.0329 -0.9993>}
 middle_01_r {parent=middle_metacarpal_r, loc=<-7.3275 0.2447 -0.0129>, rot=<-0.0208 0.0042 0.0858 -0.9961>}
 middle_02_r {parent=middle_01_r, loc=<-4.0123 -0.1401 -0.1175>, rot=<-0.0218 0.0258 0.3512 -0.9357>}
 middle_03_r {parent=middle_02_r, loc=<-2.6077 -0.0534 -0.1373>, rot=<0.0082 0.0188 0.1708 -0.9851>}
 ring_metacarpal_r {parent=hand_r, loc=<-2.2074 0.3625 -1.3638>, rot=<-0.0284 0.0722 -0.0458 -0.9959>}
 ring_01_r {parent=ring_metacarpal_r, loc=<-7.0295 0.6197 0.1446>, rot=<-0.0489 -0.0252 0.1892 -0.9804>}
 ring_02_r {parent=ring_01_r, loc=<-3.6174 -0.4647 0.1027>, rot=<0 0 0.2806 -0.9598>}
 ring_03_r {parent=ring_02_r, loc=<-2.4152 -0.114 -0.0199>, rot=<0 0 0.1189 -0.9929>}
 pinky_metacarpal_r {parent=hand_r, loc=<-1.7616 0.9753 -3.3814>, rot=<-0.0397 0.0814 -0.069 -0.9935>}
 pinky_01_r {parent=pinky_metacarpal_r, loc=<-7.1709 0.4831 0.1094>, rot=<-0.0654 -0.0035 0.2242 -0.9723>}
 pinky_02_r {parent=pinky_01_r, loc=<-2.8274 -0.0893 0.1043>, rot=<0.0037 0.0011 0.2854 -0.9584>}
 pinky_03_r {parent=pinky_02_r, loc=<-2.2046 -0.0307 0.0078>, rot=<-0.0002 0.0098 0.1445 -0.9895>}
 weapon_r {parent=hand_r, loc=<0.1974 3.4111 -0.3811>, rot=<0 0 0 1>}
 hand_attach_r {parent=hand_r, loc=<-8.5001 2.9999 0>, rot=<0 0 -0.7071 -0.7071>}
 deform_wrist_dn_base_r {parent=hand_r, loc=<-0.0001 1.5328 0>, rot=<0 0 0 -1>}
 deform_wrist_dn_r {parent=deform_wrist_dn_base_r, loc=<0 0 0>, rot=<0 0 0 1>}
 deform_wrist_up_base_r {parent=hand_r, loc=<-0.0002 -2.3007 -0.0001>, rot=<0 0 0 -1>}
 deform_wrist_up_r {parent=deform_wrist_up_base_r, loc=<0 0 0>, rot=<0 0 0 1>}
 extra_index_metacarpal_r {parent=hand_r, loc=<-2.0943 -0.9568 2.9501>, rot=<0.0158 -0.0124 -0.0173 -0.9996>}
 extra_index_01_r {parent=extra_index_metacarpal_r, loc=<-7.6893 0.4599 0.1307>, rot=<0.008 0.0473 0.093 -0.9945>}
 extra_index_02_r {parent=extra_index_01_r, loc=<-2.7531 0.2086 0.0782>, rot=<0.028 -0.0077 0.1904 -0.9813>}
 extra_index_03_r {parent=extra_index_02_r, loc=<-2.426 0.8279 0.0632>, rot=<0.0067 -0.0033 0.3108 -0.9504>}
 extra_thumb_01_r {parent=hand_r, loc=<-1.4829 0.4357 2.4305>, rot=<0.6528 -0.4271 -0.0104 -0.6256>}
 extra_thumb_02_r {parent=extra_thumb_01_r, loc=<-4.4671 0.125 0.0076>, rot=<0.0181 -0.0671 0.1178 -0.9906>}
 extra_thumb_03_r {parent=extra_thumb_02_r, loc=<-2.4171 0.5653 -0.0225>, rot=<-0.0125 0.0043 0.2921 -0.9563>}
 extra_middle_metacarpal_r {parent=hand_r, loc=<-2.2088 -1.0801 0.8207>, rot=<0.006 0.0168 -0.0329 -0.9993>}
 extra_middle_01_r {parent=extra_middle_metacarpal_r, loc=<-7.3367 0.6988 0.0756>, rot=<-0.036 0.0439 0.0853 -0.9947>}
 extra_middle_02_r {parent=extra_middle_01_r, loc=<-3.2796 0.0469 -0.0412>, rot=<0.0125 0.0039 0.2554 -0.9667>}
 extra_middle_03_r {parent=extra_middle_02_r, loc=<-2.5402 0.5119 -0.0714>, rot=<-0.029 0.0234 0.211 -0.9768>}
 extra_ring_metacarpal_r {parent=hand_r, loc=<-2.1551 -0.9116 -1.2576>, rot=<-0.0284 0.0722 -0.0458 -0.9959>}
 extra_ring_01_r {parent=extra_ring_metacarpal_r, loc=<-6.9869 1.064 0.3339>, rot=<-0.0466 0.0104 0.122 -0.9914>}
 extra_ring_02_r {parent=extra_ring_01_r, loc=<-3.1601 0.2748 0.0692>, rot=<0.0159 0.0002 0.2445 -0.9695>}
 extra_ring_03_r {parent=extra_ring_02_r, loc=<-2.2802 0.5106 0.1111>, rot=<-0.0385 0.0032 0.1927 -0.9805>}
 extra_pinky_metacarpal_r {parent=hand_r, loc=<-2.0089 -0.3195 -2.8379>, rot=<-0.0513 0.1126 0.0124 -0.9922>}
 extra_pinky_01_r {parent=extra_pinky_metacarpal_r, loc=<-6.8349 0.0746 0.2457>, rot=<-0.0547 -0.0309 0.0928 -0.9937>}
 extra_pinky_02_r {parent=extra_pinky_01_r, loc=<-2.2841 0.1128 -0.0376>, rot=<0.0104 0.0213 0.1361 -0.9904>}
 extra_pinky_03_r {parent=extra_pinky_02_r, loc=<-2.0276 0.8875 0.1063>, rot=<-0.013 -0.0106 0.299 -0.9541>}
 lowerarm_twist_02_r {parent=lowerarm_r, loc=<-10.6707 0 0>, rot=<0 0 0 -1>}
 lowerarm_twist_01_r {parent=lowerarm_r, loc=<-18.316 0 0>, rot=<0 0 0 -1>}
 dyn_bracelet_01_r {parent=lowerarm_r, loc=<-17.2744 -0.1502 -2.8108>, rot=<0.527 0.4676 0.5457 -0.4536>}
 dyn_bracelet_02_r {parent=lowerarm_r, loc=<-18.886 -0.326 -2.5186>, rot=<0.496 0.503 0.5792 -0.4068>}
 deform_elbow_in_base_r {parent=lowerarm_r, loc=<-0.1834 -3.6062 0.2572>, rot=<0 0 0 -1>}
 deform_elbow_in_r {parent=deform_elbow_in_base_r, loc=<0 0 0>, rot=<0 0 0 1>}
 deform_elbow_out_base_r {parent=lowerarm_r, loc=<0 4 -0.0001>, rot=<0 0 0 -1>}
 deform_elbow_out_r {parent=deform_elbow_out_base_r, loc=<0 0 0>, rot=<0 0 0 1>}
 upperarm_twist_01_r {parent=upperarm_r, loc=<-7.2135 0.1134 0>, rot=<0 0 0 -1>}
 deform_upperarm_bk_base_r {parent=upperarm_twist_01_r, loc=<6.4053 4.3746 -0.314>, rot=<0 0 0 -1>}
 deform_upperarm_bk_r {parent=deform_upperarm_bk_base_r, loc=<0 0 0>, rot=<0 0 0 1>}
 deform_upperarm_fr_base_r {parent=upperarm_twist_01_r, loc=<5.6312 -5.092 -0.3354>, rot=<0 0 0 -1>}
 deform_upperarm_fr_r {parent=deform_upperarm_fr_base_r, loc=<0 0 0>, rot=<0 0 0 1>}
 deform_upperarm_up_base_r {parent=upperarm_twist_01_r, loc=<8.891 -0.4825 -3.9517>, rot=<0 0 0 -1>}
 deform_upperarm_up_r {parent=deform_upperarm_up_base_r, loc=<0 0 0>, rot=<0 0 0 1>}
 upperarm_twist_02_r {parent=upperarm_r, loc=<-19.7998 0.1134 0>, rot=<0 0 0 -1>}
 deform_upperarm_r {parent=upperarm_r, loc=<-8.3421 0.598 0.783>, rot=<0.6629 -0.0065 0.7486 -0.004>}
 deform_pec_aim_base_l {parent=spine_04, loc=<3.2456 9.3309 -1.2237>, rot=<0.0603 -0.4033 0.3318 -0.8507>}
 deform_pec_aim_l {parent=deform_pec_aim_base_l, loc=<0 0 0>, rot=<0 0 0 1>}
 deform_pec_base_l {parent=deform_pec_aim_l, loc=<9.3665 0 0>, rot=<0 0 0 1>}
 deform_pec_l {parent=deform_pec_base_l, loc=<0 0 0>, rot=<0 0 0 1>}
 deform_pec_aim_base_r {parent=spine_04, loc=<3.2452 9.3309 1.2237>, rot=<-0.4033 -0.0603 0.8507 0.3318>}
 deform_pec_aim_r {parent=deform_pec_aim_base_r, loc=<0 0 0>, rot=<0 0 0 -1>}
 deform_pec_base_r {parent=deform_pec_aim_r, loc=<-9.3667 -0.0003 -0.0004>, rot=<0 0 0 1>}
 deform_pec_r {parent=deform_pec_base_r, loc=<0 0 0>, rot=<0 0 0 1>}
 dyn_jacket_base_fr_l {parent=spine_03, loc=<4.0681 10.0793 -10.4212>, rot=<0.0447 -0.9054 -0.4215 -0.0241>}
 dyn_jacket_fr_01_l {parent=dyn_jacket_base_fr_l, loc=<0 0 0>, rot=<0 0 0 1>}
 dyn_jacket_fr_02_l {parent=dyn_jacket_fr_01_l, loc=<13.1661 0.0025 0.3001>, rot=<0.0351 0.0272 -0.0105 -0.999>}
 dyn_jacket_fr_03_l {parent=dyn_jacket_fr_02_l, loc=<9.1888 -0.0854 -0.4029>, rot=<0.0318 -0.0167 -0.0031 -0.9993>}
 dyn_jacket_fr_04_l {parent=dyn_jacket_fr_03_l, loc=<11.1123 -0.0651 -0.5939>, rot=<0.029 -0.0083 -0.0093 -0.9995>}
 dyn_jacket_base_side_l {parent=spine_03, loc=<4.6869 0.2939 -14.9099>, rot=<0.0393 -0.9984 -0.0016 -0.0416>}
 dyn_jacket_side_01_l {parent=dyn_jacket_base_side_l, loc=<-0.0084 0 0.0276>, rot=<0.0001 -0.013 0.0014 -0.9999>}
 dyn_jacket_side_02_l {parent=dyn_jacket_side_01_l, loc=<12.5859 0.0356 0.7869>, rot=<0.0002 0.0499 0.0136 -0.9987>}
 dyn_jacket_side_03_l {parent=dyn_jacket_side_02_l, loc=<9.0751 0.2717 0.1934>, rot=<-0.0003 0.0053 -0.0017 -1>}
 dyn_jacket_side_04_l {parent=dyn_jacket_side_03_l, loc=<11.3756 0.0091 0.039>, rot=<0.0303 0.0208 -0.0119 -0.9993>}
 dyn_jacket_base_bk_l {parent=spine_03, loc=<6.2576 -9.5641 -11.7183>, rot=<0.0507 -0.905 0.4215 -0.026>}
 dyn_jacket_bk_01_l {parent=dyn_jacket_base_bk_l, loc=<0 -0.0005 0>, rot=<0.0051 0 -0.0004 -1>}
 dyn_jacket_bk_02_l {parent=dyn_jacket_bk_01_l, loc=<13.0879 0.254 -0.0125>, rot=<-0.0057 0.0511 -0.005 -0.9987>}
 dyn_jacket_bk_03_l {parent=dyn_jacket_bk_02_l, loc=<9.0426 0.288 -0.3176>, rot=<-0.0243 0.0228 -0.0122 -0.9994>}
 dyn_jacket_bk_04_l {parent=dyn_jacket_bk_03_l, loc=<11.5749 0.0004 0.0433>, rot=<-0.0026 0.0219 -0.0238 -0.9995>}
 dyn_jacket_base_mid {parent=spine_03, loc=<6.8018 -12.2748 -0.0409>, rot=<0.0535 -0.7097 0.7005 -0.0528>}
 dyn_jacket_mid_01 {parent=dyn_jacket_base_mid, loc=<0 -0.0409 -0.0005>, rot=<0 -0.004 0 -1>}
 dyn_jacket_mid_02 {parent=dyn_jacket_mid_01, loc=<13.4554 -0.0023 0.285>, rot=<0 0.0155 0 -0.9999>}
 dyn_jacket_mid_03 {parent=dyn_jacket_mid_02, loc=<8.8043 -0.0048 0.1659>, rot=<0 -0.0006 0 -1>}
 dyn_jacket_mid_04 {parent=dyn_jacket_mid_03, loc=<13.1536 -0.0056 0.1443>, rot=<0 -0.0087 0 -1>}
 dyn_jacket_base_bk_r {parent=spine_03, loc=<6.2572 -9.5641 11.7183>, rot=<0.026 -0.4215 0.905 -0.0507>}
 dyn_jacket_bk_01_r {parent=dyn_jacket_base_bk_r, loc=<0 0.0005 0>, rot=<0.0051 0 -0.0004 -1>}
 dyn_jacket_bk_02_r {parent=dyn_jacket_bk_01_r, loc=<13.0867 -0.2742 -0.0179>, rot=<-0.0046 0.0126 0.0061 -0.9999>}
 dyn_jacket_bk_03_r {parent=dyn_jacket_bk_02_r, loc=<9.0401 -0.2879 0.3793>, rot=<0.0233 0.0228 0.014 -0.9994>}
 dyn_jacket_bk_04_r {parent=dyn_jacket_bk_03_r, loc=<11.5378 -0.0004 0.9339>, rot=<0.0008 0.0219 0.024 -0.9995>}
 dyn_jacket_base_side_r {parent=spine_03, loc=<4.6872 0.2939 14.9099>, rot=<0.0416 0.0016 0.9984 -0.0393>}
 dyn_jacket_side_01_r {parent=dyn_jacket_base_side_r, loc=<0 0 0>, rot=<0 0 0 1>}
 dyn_jacket_side_02_r {parent=dyn_jacket_side_01_r, loc=<12.5942 0 0.486>, rot=<0 -0.0017 -0.015 -0.9999>}
 dyn_jacket_side_03_r {parent=dyn_jacket_side_02_r, loc=<9.0337 -0.2717 0.8912>, rot=<0.0001 0.0053 0.0017 -1>}
 dyn_jacket_side_04_r {parent=dyn_jacket_side_03_r, loc=<11.3383 -0.0091 0.9144>, rot=<-0.0312 0.0208 0.0096 -0.9993>}
 dyn_jacket_base_fr_r {parent=spine_03, loc=<4.0679 10.0793 10.4212>, rot=<0.0241 0.4215 0.9054 -0.0447>}
 dyn_jacket_fr_01_r {parent=dyn_jacket_base_fr_r, loc=<0 0 0>, rot=<0 0 0 1>}
 dyn_jacket_fr_02_r {parent=dyn_jacket_fr_01_r, loc=<13.1662 -0.0026 0.3001>, rot=<-0.0354 -0.0113 0.0091 -0.9993>}
 dyn_jacket_fr_03_r {parent=dyn_jacket_fr_02_r, loc=<9.1923 0.0854 0.3055>, rot=<-0.032 -0.0167 0.0007 -0.9993>}
 dyn_jacket_fr_04_r {parent=dyn_jacket_fr_03_r, loc=<11.1249 0.0651 0.2631>, rot=<-0.0296 -0.0083 0.007 -0.9995>}
 deform_jacket_root {parent=spine_02, loc=<0 0 0>, rot=<0 0 0 1>}
 deform_strap_root {parent=spine_01, loc=<3.8759 -0.1722 10.7905>, rot=<0.1736 0.0028 0.9847 -0.0159>}
 deform_strap_01 {parent=deform_strap_root, loc=<0.0569 0 -29.2348>, rot=<0 -0.2172 0 -0.9761>}
 thigh_l {parent=pelvis, loc=<-1.0855 -0.2109 -8.7053>, rot=<0.0009 0.0355 0.0155 -0.9993>}
 calf_l {parent=thigh_l, loc=<-42.5284 -0.0019 -0.7297>, rot=<0.0006 -0.0161 -0.0391 -0.9991>}
 foot_l {parent=calf_l, loc=<-46.7401 -2.5014 -0.592>, rot=<0.0427 -0.0238 0.0353 -0.9982>}
 ball_l {parent=foot_l, loc=<-3.5658 10.8066 -0.1544>, rot=<0.0001 -0.0001 -0.7063 -0.7079>}
 calf_twist_02_l {parent=calf_l, loc=<-19.1546 -1.7943 -0.2638>, rot=<0.0237 0.0095 -0.0035 -0.9997>}
 calf_twist_01_l {parent=calf_l, loc=<-35.8483 -1.9197 -0.5794>, rot=<0.0237 0.0095 -0.0035 -0.9997>}
 dyn_zip_l {parent=calf_l, loc=<-26.9969 -2.4026 4.2741>, rot=<0.0199 0.0652 0.9969 0.0398>}
 extra_foot_l {parent=calf_l, loc=<-40.4635 -2.7516 -0.5937>, rot=<0.016 0.0215 0.9988 0.0405>}
 extra_ball_l {parent=extra_foot_l, loc=<10.7632 -6.4085 -0.5803>, rot=<-0.0022 -0.0215 0.9998 0.0001>}
 deform_knee_in_base_l {parent=calf_l, loc=<0.8031 -5.0231 0.2456>, rot=<0 0 0 -1>}
 deform_knee_in_l {parent=deform_knee_in_base_l, loc=<0 0 0>, rot=<0 0 0 1>}
 deform_knee_out_base_l {parent=calf_l, loc=<0 4 0>, rot=<0 0 0 -1>}
 deform_knee_out_l {parent=deform_knee_out_base_l, loc=<0 0 0>, rot=<0 0 0 1>}
 thigh_twist_01_l {parent=thigh_l, loc=<-27.952 0 0>, rot=<0.0013 0 0 -1>}
 deform_groin_base_l {parent=thigh_l, loc=<0.4801 5.6776 -0.2486>, rot=<0 0 0 -1>}
 deform_groin_l {parent=deform_groin_base_l, loc=<0 0 0>, rot=<0 0 0 1>}
 thigh_r {parent=pelvis, loc=<-1.0855 -0.2109 8.7053>, rot=<0.0355 -0.0009 0.9993 0.0155>}
 calf_r {parent=thigh_r, loc=<42.5284 0.0019 0.7297>, rot=<0.0006 -0.0161 -0.0391 -0.9991>}
 foot_r {parent=calf_r, loc=<46.7401 2.5014 0.592>, rot=<0.0427 -0.0238 0.0353 -0.9982>}
 ball_r {parent=foot_r, loc=<3.5658 -10.8066 0.1544>, rot=<0.0001 -0.0001 -0.7063 -0.7079>}
 calf_twist_02_r {parent=calf_r, loc=<19.1546 1.7943 0.2638>, rot=<0.0237 0.0095 -0.0035 -0.9997>}
 calf_twist_01_r {parent=calf_r, loc=<35.8483 1.9197 0.5794>, rot=<0.0237 0.0095 -0.0035 -0.9997>}
 dyn_zip_r {parent=calf_r, loc=<26.9969 2.4026 -4.2741>, rot=<0.0652 -0.0199 0.0398 -0.9969>}
 extra_foot_r {parent=calf_r, loc=<40.4634 2.7516 0.5938>, rot=<0.0215 -0.016 0.0405 -0.9988>}
 extra_ball_r {parent=extra_foot_r, loc=<10.7632 -6.4084 0.5803>, rot=<0.0215 -0.0022 0.0001 -0.9998>}
 deform_knee_in_base_r {parent=calf_r, loc=<-0.8031 5.0231 -0.2457>, rot=<0 0 0 -1>}
 deform_knee_in_r {parent=deform_knee_in_base_r, loc=<0 0 0>, rot=<0 0 0 1>}
 deform_knee_out_base_r {parent=calf_r, loc=<0 -4 0>, rot=<0 0 0 -1>}
 deform_knee_out_r {parent=deform_knee_out_base_r, loc=<0 0 0>, rot=<0 0 0 1>}
 thigh_twist_01_r {parent=thigh_r, loc=<27.952 0 0>, rot=<0.0013 0 0 -1>}
 deform_groin_base_r {parent=thigh_r, loc=<-0.4801 -5.6776 0.2486>, rot=<0 0 0 -1>}
 deform_groin_r {parent=deform_groin_base_r, loc=<0 0 0>, rot=<0 0 0 1>}
 deform_glute_base_l {parent=pelvis, loc=<-7.3466 -9.0213 -9.1245>, rot=<0.0172 0.031 -0.4477 -0.8935>}
 deform_glute_l {parent=deform_glute_base_l, loc=<0 0 0>, rot=<0 0 0 1>}
 deform_glute_base_r {parent=pelvis, loc=<-7.3464 -9.0213 9.1245>, rot=<0.031 -0.0172 0.8935 -0.4477>}
 deform_glute_r {parent=deform_glute_base_r, loc=<0 0 0>, rot=<0 0 0 1>}
 deform_belt_base {parent=pelvis, loc=<3.741 8.3618 -8.1391>, rot=<0.1407 0.1221 0.9819 0.0338>}
 deform_belt_01 {parent=deform_belt_base, loc=<0 0 0>, rot=<0 0 0 1>}
 dyn_strap_base {parent=deform_belt_base, loc=<0 0 0>, rot=<0 0 0 1>}
 dyn_strap_01 {parent=dyn_strap_base, loc=<0 0 0>, rot=<0 0 0 1>}
 dyn_strap_02 {parent=dyn_strap_01, loc=<4.4918 -0.0833 -0.5692>, rot=<-0.0214 -0.1393 0.0044 -0.99>}
 dyn_strap_03 {parent=dyn_strap_02, loc=<8.5875 -0.2574 0.114>, rot=<0 0 0 1>}
 dyn_strap_04 {parent=dyn_strap_03, loc=<7.5889 0 0>, rot=<0 0 0 1>}
 dyn_strap_05 {parent=dyn_strap_04, loc=<8.0454 0.4229 0.0808>, rot=<0.0178 0 0 -0.9998>}
 deform_belt_side_l {parent=pelvis, loc=<-0.175 -1.5472 -16.6909>, rot=<-0.009 -0.9937 0.0057 -0.1115>}
]
variables [IsPlayEmote {type=bool, default=False}]
