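# Physics asset for the F_MED_RenegadeWhip female body, setting up the jacket's lower section as jointed rigid bodies around the hips and legs.
# Seven bone chains (front, side and back pairs plus a middle) hang from small pelvis-linked spheres through capsules with angular constraints.
# Cross-constraints tie many neighboring chains into a cloth-like lattice, with pelvis and thigh capsules as body volumes.
# Setup is asymmetric: right-side base joints swing and twist freely, left-side ones are limited.
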PhysicsAsset /BRCosmetics/Characters/Player/Female/Medium/Bodies/F_MED_RenegadeWhip/Meshes/F_MED_RenegadeWhip_Physics
bodies {
 dyn_jacket_base_bk_l [{shapeType=Sphere, radius=0.5, shapeName=dyn_jacket_base_bk_l_sphere}]
 dyn_jacket_base_bk_r [{shapeType=Sphere, radius=0.5, shapeName=dyn_jacket_base_bk_r_sphere}]
 dyn_jacket_base_fr_l [{shapeType=Sphere, radius=0.5, shapeName=dyn_jacket_base_fr_l_sphere}]
 dyn_jacket_base_fr_r [{shapeType=Sphere, radius=0.5, shapeName=dyn_jacket_base_fr_r_sphere}]
 dyn_jacket_base_mid [{shapeType=Sphere, radius=0.5, shapeName=dyn_jacket_base_mid_sphere}]
 dyn_jacket_base_side_l [{shapeType=Sphere, radius=0.5, shapeName=dyn_jacket_base_side_l_sphere}]
 dyn_jacket_base_side_r [{shapeType=Sphere, radius=0.5, shapeName=dyn_jacket_base_side_r_sphere}]
 dyn_jacket_bk_01_l [{shapeType=Capsule, length=28.081, radius=7, shapeName=dyn_jacket_fr_04_r_capsule, center=<18.339 0 5.9415>, rotation=<rot 90 0 0>}]
 dyn_jacket_bk_01_r [{shapeType=Capsule, length=28.081, radius=7, shapeName=dyn_jacket_fr_04_r_capsule, center=<18.339 0 5.9415>, rotation=<rot 90 0 0>}]
 dyn_jacket_bk_03_l [{shapeType=Capsule, length=22, radius=6, shapeName=dyn_jacket_fr_04_r_capsule, center=<3.9312 0 1.185>, rotation=<rot 0 0 90>}]
 dyn_jacket_bk_03_r [{shapeType=Capsule, length=22, radius=6, shapeName=dyn_jacket_fr_04_r_capsule, center=<3.9312 0 1.185>, rotation=<rot 0 0 90>}]
 dyn_jacket_bk_04_l [{shapeType=Capsule, length=22, radius=6, shapeName=dyn_jacket_fr_04_r_capsule, center=<3.7772 0 -0.0925>, rotation=<rot 0 0 90>}]
 dyn_jacket_bk_04_r [{shapeType=Capsule, length=22, radius=6, shapeName=dyn_jacket_fr_04_r_capsule, center=<3.7772 0 -0.0925>, rotation=<rot 0 0 90>}]
 dyn_jacket_fr_01_l [{shapeType=Capsule, length=16, radius=6, shapeName=dyn_jacket_fr_01_l_capsule, center=<13.2587 2 3.2314>, rotation=<rot 90 0 0>}]
 dyn_jacket_fr_01_r [{shapeType=Capsule, length=16, radius=6, shapeName=dyn_jacket_fr_01_l_capsule, center=<13.2587 -2 3.2314>, rotation=<rot 90 0 0>}]
 dyn_jacket_fr_03_l [{shapeType=Capsule, length=12, radius=6, shapeName=dyn_jacket_fr_03_l_capsule, center=<7.4263 2 1.5825>, rotation=<rot 87.2185 0 -0.0001>}]
 dyn_jacket_fr_03_r [{shapeType=Capsule, length=12, radius=6, shapeName=dyn_jacket_fr_04_r_capsule, center=<7.4263 -2 2.5825>, rotation=<rot 87.2185 0 0>}]
 dyn_jacket_fr_04_l [{shapeType=Capsule, length=25, radius=6, shapeName=dyn_jacket_fr_04_l_capsule, center=<3.0457 0 1.6415>, rotation=<rot 0 0 90>}]
 dyn_jacket_fr_04_r [{shapeType=Capsule, length=25, radius=6, shapeName=dyn_jacket_fr_04_r_capsule, center=<3.0457 0 1.6415>, rotation=<rot 0 0 90>}]
 dyn_jacket_mid_01 [{shapeType=Capsule, length=25, radius=6, shapeName=dyn_jacket_mid_01_capsule, center=<16.7819 0 5.8315>, rotation=<rot 90 0 0>}]
 dyn_jacket_mid_03 [{shapeType=Capsule, length=20, radius=6, shapeName=dyn_jacket_mid_03_capsule, center=<4.0129 0 3.4236>, rotation=<rot 0 0 90>}]
 dyn_jacket_mid_04 [{shapeType=Capsule, length=20, radius=7, shapeName=dyn_jacket_mid_04_capsule, center=<5.1707 0 0.5959>, rotation=<rot 0 0 90>}]
 dyn_jacket_side_01_l [{shapeType=Capsule, length=12, radius=7, shapeName=dyn_jacket_fr_04_r_capsule, center=<11.4075 0 5.599>, rotation=<rot 95 0 0>}]
 dyn_jacket_side_01_r [{shapeType=Capsule, length=12, radius=7, shapeName=dyn_jacket_fr_04_r_capsule, center=<11.4075 0 4.8381>, rotation=<rot 95 0 0>}]
 dyn_jacket_side_03_l [{shapeType=Capsule, length=20, radius=6, shapeName=dyn_jacket_fr_04_r_capsule, center=<3.7386 0 4.8825>, rotation=<rot 0 0 90>}]
 dyn_jacket_side_03_r [{shapeType=Capsule, length=20, radius=6, shapeName=dyn_jacket_fr_04_r_capsule, center=<3.7386 0 4.5894>, rotation=<rot 0 0 90>}]
 dyn_jacket_side_04_l [{shapeType=Capsule, length=20, radius=6, shapeName=dyn_jacket_fr_04_r_capsule, center=<4.4287 0 3.3484>, rotation=<rot 0 0 90>}]
 dyn_jacket_side_04_r [{shapeType=Capsule, length=20, radius=6, shapeName=dyn_jacket_fr_04_r_capsule, center=<4.4287 0 3.3484>, rotation=<rot 0 0 90>}]
 pelvis [{shapeType=Capsule, length=7.6669, radius=14.1213, shapeName=pelvis_capsule, center=<0.6596 -1.8886 0>, rotation=<rot 90 0 0>}]
 thigh_l [
  {
   shapeType=Capsule
   length=45.6725
   radius=9.071
   shapeName=thigh_l_capsule
   center=<-27.4438 -2.0902 -0.9357>
   rotation=<rot 90 171.1193 167.5007>
  }
  {shapeType=Capsule, length=10.2, radius=9.5, shapeName=thigh_l_capsule, center=<-15.9616 0 8.3933>, rotation=<rot 87.8291 0 0.0001>}
 ]
 thigh_r [{shapeType=Capsule, length=30.8683, radius=8.746, shapeName=thigh_r_capsule, center=<20.4052 1.0735 0.8166>, rotation=<rot 90 180 177.586>}]
}
constraints [
 {bone1Name=dyn_jacket_mid_03, bone2Name=dyn_jacket_mid_01, swing1LimitDegrees=5, swing2LimitDegrees=45, twistLimitDegrees=5}
 {bone1Name=dyn_jacket_mid_04, bone2Name=dyn_jacket_mid_03, swing1LimitDegrees=5, swing2LimitDegrees=65, twistLimitDegrees=5}
 {bone1Name=dyn_jacket_mid_01, bone2Name=dyn_jacket_base_mid, swing1LimitDegrees=5, swing2LimitDegrees=45, twistLimitDegrees=5}
 {bone1Name=dyn_jacket_base_mid, bone2Name=pelvis, swing1LimitDegrees=10, swing2LimitDegrees=45, twistLimitDegrees=5}
 {bone1Name=dyn_jacket_side_03_r, bone2Name=dyn_jacket_side_01_r, swing1LimitDegrees=5, swing2LimitDegrees=45, twistLimitDegrees=5}
 {bone1Name=dyn_jacket_side_04_r, bone2Name=dyn_jacket_side_03_r, swing1LimitDegrees=5, swing2LimitDegrees=45, twistLimitDegrees=5}
 {bone1Name=dyn_jacket_fr_04_r, bone2Name=dyn_jacket_fr_03_r, swing1LimitDegrees=5, swing2LimitDegrees=55, twistLimitDegrees=20}
 {bone1Name=dyn_jacket_fr_03_r, bone2Name=dyn_jacket_fr_01_r, swing1LimitDegrees=35, swing2LimitDegrees=55, twistLimitDegrees=25}
 {bone1Name=dyn_jacket_bk_04_r, bone2Name=dyn_jacket_bk_03_r, swing1LimitDegrees=10, swing2LimitDegrees=45, twistLimitDegrees=5}
 {bone1Name=dyn_jacket_bk_03_r, bone2Name=dyn_jacket_bk_01_r, swing1LimitDegrees=10, swing2LimitDegrees=45, twistLimitDegrees=5}
 {bone1Name=dyn_jacket_fr_04_l, bone2Name=dyn_jacket_fr_03_l, swing1LimitDegrees=10, swing2LimitDegrees=45, twistLimitDegrees=25}
 {bone1Name=dyn_jacket_base_fr_l, bone2Name=pelvis, swing1LimitDegrees=10, swing2LimitDegrees=45, twistLimitDegrees=5}
 {bone1Name=dyn_jacket_fr_01_l, bone2Name=dyn_jacket_base_fr_l, swing1LimitDegrees=10, swing2LimitDegrees=35, twistLimitDegrees=25}
 {bone1Name=dyn_jacket_fr_03_l, bone2Name=dyn_jacket_fr_01_l, swing1LimitDegrees=15, swing2LimitDegrees=45, twistLimitDegrees=25}
 {bone1Name=dyn_jacket_bk_04_l, bone2Name=dyn_jacket_bk_03_l, swing1LimitDegrees=10, swing2LimitDegrees=45, twistLimitDegrees=5}
 {bone1Name=dyn_jacket_bk_03_l, bone2Name=dyn_jacket_bk_01_l, swing1LimitDegrees=10, swing2LimitDegrees=45, twistLimitDegrees=5}
 {bone1Name=dyn_jacket_bk_01_l, bone2Name=dyn_jacket_base_bk_l, swing1LimitDegrees=5, swing2LimitDegrees=45, twistLimitDegrees=5}
 {bone1Name=dyn_jacket_base_bk_l, bone2Name=pelvis, swing1LimitDegrees=10, swing2LimitDegrees=45, twistLimitDegrees=5}
 {bone1Name=dyn_jacket_side_04_l, bone2Name=dyn_jacket_side_03_l, swing1LimitDegrees=5, swing2LimitDegrees=45, twistLimitDegrees=5}
 {bone1Name=dyn_jacket_side_03_l, bone2Name=dyn_jacket_side_01_l, swing1LimitDegrees=5, swing2LimitDegrees=45, twistLimitDegrees=5}
 {bone1Name=dyn_jacket_side_01_l, bone2Name=dyn_jacket_base_side_l, swing1LimitDegrees=5, swing2LimitDegrees=45, twistLimitDegrees=5}
 {bone1Name=dyn_jacket_base_side_l, bone2Name=pelvis, swing1LimitDegrees=10, swing2LimitDegrees=45, twistLimitDegrees=5}
 {bone1Name=dyn_jacket_side_03_l, bone2Name=dyn_jacket_fr_03_l, swing1LimitDegrees=45, swing2LimitDegrees=45, twistLimitDegrees=45}
 {bone1Name=dyn_jacket_side_04_r, bone2Name=dyn_jacket_fr_04_r, swing1LimitDegrees=45, swing2LimitDegrees=45, twistLimitDegrees=45}
 {
  bone1Name=dyn_jacket_mid_03
  bone2Name=dyn_jacket_bk_03_l
  swing1LimitDegrees=45
  swing1Motion=Locked
  swing2LimitDegrees=45
  swing2Motion=Free
  twistLimitDegrees=45
  twistMotion=Free
 }
 {bone1Name=dyn_jacket_bk_03_l, bone2Name=dyn_jacket_side_03_l, swing1LimitDegrees=15, swing2LimitDegrees=15, twistLimitDegrees=45}
 {bone1Name=dyn_jacket_fr_01_r, bone2Name=dyn_jacket_base_fr_r, swing1LimitDegrees=20, swing2LimitDegrees=45, twistLimitDegrees=35}
 {bone1Name=dyn_jacket_bk_01_r, bone2Name=dyn_jacket_base_bk_r, swing1LimitDegrees=5, swing2LimitDegrees=40, twistLimitDegrees=5}
 {bone1Name=dyn_jacket_side_01_r, bone2Name=dyn_jacket_base_side_r, swing1LimitDegrees=5, swing2LimitDegrees=45, twistLimitDegrees=5}
 {
  bone1Name=dyn_jacket_base_bk_r
  bone2Name=pelvis
  swing1LimitDegrees=45
  swing1Motion=Free
  swing2LimitDegrees=45
  swing2Motion=Free
  twistLimitDegrees=45
  twistMotion=Free
 }
 {
  bone1Name=dyn_jacket_base_fr_r
  bone2Name=pelvis
  swing1LimitDegrees=45
  swing1Motion=Free
  swing2LimitDegrees=45
  swing2Motion=Free
  twistLimitDegrees=45
  twistMotion=Free
 }
 {
  bone1Name=dyn_jacket_base_side_r
  bone2Name=pelvis
  swing1LimitDegrees=45
  swing1Motion=Free
  swing2LimitDegrees=45
  swing2Motion=Free
  twistLimitDegrees=45
  twistMotion=Free
 }
 {bone1Name=dyn_jacket_bk_04_r, bone2Name=dyn_jacket_mid_04, swing1LimitDegrees=15, swing2LimitDegrees=15, twistLimitDegrees=15}
 {bone1Name=dyn_jacket_bk_04_l, bone2Name=dyn_jacket_side_04_l, swing1LimitDegrees=15, swing2LimitDegrees=15, twistLimitDegrees=15}
 {bone1Name=dyn_jacket_side_04_l, bone2Name=dyn_jacket_fr_04_l, swing1LimitDegrees=45, swing2LimitDegrees=45, twistLimitDegrees=45}
 {bone1Name=dyn_jacket_mid_04, bone2Name=dyn_jacket_bk_04_l, swing1LimitDegrees=15, swing2LimitDegrees=15, twistLimitDegrees=15}
 {
  bone1Name=dyn_jacket_side_03_r
  bone2Name=dyn_jacket_bk_03_r
  swing1LimitDegrees=15
  swing1Motion=Free
  swing2LimitDegrees=20
  swing2Motion=Free
  twistLimitDegrees=30
  twistMotion=Free
 }
 {bone1Name=dyn_jacket_side_04_r, bone2Name=dyn_jacket_bk_04_r, swing1LimitDegrees=45, swing2LimitDegrees=45, twistLimitDegrees=45}
 {
  bone1Name=dyn_jacket_bk_01_l
  bone2Name=dyn_jacket_side_01_l
  swing1LimitDegrees=45
  swing1Motion=Free
  swing2LimitDegrees=45
  swing2Motion=Free
  twistLimitDegrees=45
  twistMotion=Free
 }
 {
  bone1Name=dyn_jacket_side_01_l
  bone2Name=dyn_jacket_fr_01_l
  swing1LimitDegrees=5
  swing1Motion=Free
  swing2LimitDegrees=5
  swing2Motion=Free
  twistLimitDegrees=5
  twistMotion=Free
 }
 {
  bone1Name=dyn_jacket_side_03_r
  bone2Name=dyn_jacket_fr_03_r
  swing1LimitDegrees=45
  swing1Motion=Free
  swing2LimitDegrees=45
  swing2Motion=Free
  twistLimitDegrees=45
  twistMotion=Free
 }
 {
  bone1Name=dyn_jacket_bk_03_l
  bone2Name=dyn_jacket_bk_04_l
  swing1LimitDegrees=45
  swing1Motion=Free
  swing2LimitDegrees=45
  swing2Motion=Free
  twistLimitDegrees=45
  twistMotion=Free
 }
 {
  bone1Name=thigh_r
  bone2Name=pelvis
  swing1LimitDegrees=45
  swing1Motion=Free
  swing2LimitDegrees=45
  swing2Motion=Free
  twistLimitDegrees=45
  twistMotion=Free
 }
 {
  bone1Name=thigh_l
  bone2Name=pelvis
  swing1LimitDegrees=45
  swing1Motion=Free
  swing2LimitDegrees=45
  swing2Motion=Free
  twistLimitDegrees=45
  twistMotion=Free
 }
]
