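# Physics asset for the strap mesh of the RenegadeWhip female medium player body, covering a five-bone dynamic strap and two left front jacket bones.
# Strap bodies chain from the pelvis, with angular limits between segments; the pelvis attachment, jacket, thigh and root joints swing and twist freely.
# Capsules on pelvis and thighs give body volume around the strap.
# A huge root sphere, offset downward to top out near the origin, approximates a floor.
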PhysicsAsset /BRCosmetics/Characters/Player/Female/Medium/Bodies/F_MED_RenegadeWhip/Meshes/F_MED_RenegadeWhip_Strap_Physics
bodies {
 dyn_jacket_fr_03_l [{
   shapeType=Capsule
   length=7.7921
   radius=1.7622
   shapeName=dyn_jacket_fr_04_l_capsule
   center=<5.5818 3.2088 -0.1959>
   rotation=<rot 3.7253 -89.6038 -86.9784>
  }]
 dyn_jacket_fr_04_l [{
   shapeType=Capsule
   length=6.9458
   radius=2.0045
   shapeName=dyn_jacket_fr_04_l_capsule
   center=<4.4698 3.1702 -0.4841>
   rotation=<rot -7.0093 91.0917 -94.5083>
  }]
 dyn_strap_01 [{shapeType=Sphere, radius=2.105, shapeName=dyn_strap_01_sphere}]
 dyn_strap_02 [
  {
   shapeType=Capsule
   length=0.8
   radius=3.305
   shapeName=dyn_strap_03_capsule
   center=<4.5444 -1.1025 0.1324>
   rotation=<rot 0.7088 -2.508 178.9528>
  }
  {shapeType=Capsule, length=1, radius=3.3, shapeName=dyn_strap_02_capsule, center=<0 -0.963 0.1175>}
 ]
 dyn_strap_03 [{
   shapeType=Capsule
   length=1
   radius=3.205
   shapeName=dyn_strap_03_capsule
   center=<3.5781 -0.3182 0.0349>
   rotation=<rot 0.1375 -2.4977 178.9529>
  }]
 dyn_strap_04 [{
   shapeType=Capsule
   length=0.7
   radius=3.205
   shapeName=dyn_strap_03_capsule
   center=<3.9066 -0.1147 0.053>
   rotation=<rot 0.9503 -2.5124 178.9527>
  }]
 dyn_strap_05 [{
   shapeType=Capsule
   length=2.6
   radius=2.905
   shapeName=dyn_strap_03_capsule
   center=<2.9244 -0.4966 0.0292>
   rotation=<rot 0.9503 -2.5124 178.9527>
  }]
 pelvis [
  {
   shapeType=Capsule
   length=20.0642
   radius=10.8696
   shapeName=pelvis_capsule
   center=<-5.4682 -2.4492 0.1401>
   rotation=<rot 2.1542 -91.7148 1.414>
  }
  {shapeType=Capsule, length=2.1, radius=4.8, shapeName=pelvis_capsule, center=<-6.6448 4.7787 0>}
 ]
 root [{shapeType=Sphere, radius=576.7053, shapeName=root_sphere, center=<0 0 -574.9434>}]
 thigh_l [{
   shapeType=Capsule
   length=36.6011
   radius=8.2376
   shapeName=thigh_l_capsule
   center=<-19.6435 -1.007 -0.535>
   rotation=<rot -4.4848 -84.332 -93.8781>
  }]
 thigh_r [
  {
   shapeType=Capsule
   length=28.0742
   radius=6.9392
   shapeName=thigh_r_capsule
   center=<23.6187 0.6119 0.244>
   rotation=<rot 0.3507 -82.8644 -97.3574>
  }
  {shapeType=Capsule, length=17, radius=6, shapeName=thigh_r_capsule, center=<11.4872 -1.7163 -2.5381>, rotation=<rot 84.091 179.9995 172.3142>}
 ]
}
constraints [
 {bone1Name=dyn_strap_05, bone2Name=dyn_strap_04, swing1LimitDegrees=35, swing2LimitDegrees=15, twistLimitDegrees=25}
 {bone1Name=dyn_strap_02, bone2Name=dyn_strap_01, swing1LimitDegrees=45, swing2LimitDegrees=40, twistLimitDegrees=35}
 {bone1Name=dyn_strap_03, bone2Name=dyn_strap_02, swing1LimitDegrees=35, swing2LimitDegrees=15, twistLimitDegrees=25}
 {bone1Name=dyn_strap_04, bone2Name=dyn_strap_03, swing1LimitDegrees=35, swing2LimitDegrees=15, twistLimitDegrees=25}
 {
  bone1Name=thigh_l
  bone2Name=pelvis
  swing1LimitDegrees=45
  swing1Motion=Free
  swing2LimitDegrees=45
  swing2Motion=Free
  twistLimitDegrees=45
  twistMotion=Free
 }
 {
  bone1Name=thigh_r
  bone2Name=pelvis
  swing1LimitDegrees=45
  swing1Motion=Free
  swing2LimitDegrees=45
  swing2Motion=Free
  twistLimitDegrees=45
  twistMotion=Free
 }
 {
  bone1Name=dyn_strap_01
  bone2Name=pelvis
  swing1LimitDegrees=45
  swing1Motion=Free
  swing2LimitDegrees=45
  swing2Motion=Free
  twistLimitDegrees=45
  twistMotion=Free
 }
 {
  bone1Name=pelvis
  bone2Name=root
  swing1LimitDegrees=45
  swing1Motion=Free
  swing2LimitDegrees=45
  swing2Motion=Free
  twistLimitDegrees=45
  twistMotion=Free
 }
 {
  bone1Name=dyn_jacket_fr_04_l
  bone2Name=dyn_jacket_fr_03_l
  swing1LimitDegrees=45
  swing1Motion=Free
  swing2LimitDegrees=45
  swing2Motion=Free
  twistLimitDegrees=45
  twistMotion=Free
 }
 {
  bone1Name=dyn_jacket_fr_03_l
  bone2Name=pelvis
  swing1LimitDegrees=45
  swing1Motion=Free
  swing2LimitDegrees=45
  swing2Motion=Free
  twistLimitDegrees=45
  twistMotion=Free
 }
]
