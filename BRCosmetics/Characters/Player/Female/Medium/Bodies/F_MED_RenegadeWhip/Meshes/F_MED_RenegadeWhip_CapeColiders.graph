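# Physics asset for the F_MED_RenegadeWhip cape, simulating seven dyn_jacket bone chains (front, side and back per side, plus middle) against body colliders.
# Each chain runs four capsules from a small base sphere constrained to spine_03, mostly with limited swing and twist.
# Cross-constraints tie neighboring chains into a sheet.
# Spine, pelvis, leg and foot capsules have free constraints, and a huge root sphere below apparently serves as a floor.
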PhysicsAsset /BRCosmetics/Characters/Player/Female/Medium/Bodies/F_MED_RenegadeWhip/Meshes/F_MED_RenegadeWhip_CapeColiders
bodies {
 calf_l [{
   shapeType=Capsule
   length=30.5899
   radius=6.2404
   shapeName=calf_l_capsule
   center=<-19.944 -3.2896 -0.1744>
   rotation=<rot -0.0268 -86.7447 -89.9339>
  }]
 calf_r [{
   shapeType=Capsule
   length=30.4607
   radius=6.6253
   shapeName=calf_r_capsule
   center=<19.9079 3.3019 -0.1195>
   rotation=<rot -0.0957 -88.1314 -92.1944>
  }]
 dyn_jacket_base_bk_l [{shapeType=Sphere, radius=0.5, shapeName=dyn_jacket_base_bk_l_sphere}]
 dyn_jacket_base_bk_r [{shapeType=Sphere, radius=0.5, shapeName=dyn_jacket_base_bk_r_sphere}]
 dyn_jacket_base_fr_l [{shapeType=Sphere, radius=0.5, shapeName=dyn_jacket_base_fr_l_sphere}]
 dyn_jacket_base_fr_r [{shapeType=Sphere, radius=0.5, shapeName=dyn_jacket_base_fr_r_sphere}]
 dyn_jacket_base_mid [{shapeType=Sphere, radius=0.5, shapeName=dyn_jacket_base_mid_sphere}]
 dyn_jacket_base_side_l [{shapeType=Sphere, radius=0.5, shapeName=dyn_jacket_base_side_l_sphere}]
 dyn_jacket_base_side_r [{shapeType=Sphere, radius=0.5, shapeName=dyn_jacket_base_side_r_sphere}]
 dyn_jacket_bk_01_l [{
   shapeType=Capsule
   length=9.1516
   radius=2.8992
   shapeName=dyn_jacket_fr_04_r_capsule
   center=<8.4081 -0.1043 -1.4721>
   rotation=<rot -0.361 -0.0858 -91.1397>
  }]
 dyn_jacket_bk_01_r [{
   shapeType=Capsule
   length=9.2516
   radius=2.7992
   shapeName=dyn_jacket_fr_04_r_capsule
   center=<8.4483 -0.5983 1.2086>
   rotation=<rot -0.3561 -0.09 79.4412>
  }]
 dyn_jacket_bk_02_l [{
   shapeType=Capsule
   length=6.8516
   radius=4.9992
   shapeName=dyn_jacket_fr_04_r_capsule
   center=<3.7232 -0.1241 -2.2562>
   rotation=<rot -0.4167 -0.4146 -90.2739>
  }]
 dyn_jacket_bk_02_r [{
   shapeType=Capsule
   length=6.1516
   radius=5.1992
   shapeName=dyn_jacket_fr_04_r_capsule
   center=<3.9354 -0.532 0.5365>
   rotation=<rot 0.3867 1.2098 -100.4285>
  }]
 dyn_jacket_bk_03_l [{
   shapeType=Capsule
   length=8.8516
   radius=6.1992
   shapeName=dyn_jacket_fr_04_r_capsule
   center=<5.6372 0.5969 -1.9412>
   rotation=<rot -0.3728 -0.1017 -91.1806>
  }]
 dyn_jacket_bk_03_r [{
   shapeType=Capsule
   length=7.5516
   radius=6.1992
   shapeName=dyn_jacket_fr_04_r_capsule
   center=<5.3919 -0.1565 0.9911>
   rotation=<rot 0.279 1.0335 -96.9521>
  }]
 dyn_jacket_bk_04_l [{
   shapeType=Capsule
   length=8.5516
   radius=7.8992
   shapeName=dyn_jacket_fr_04_r_capsule
   center=<4.6321 0.4278 -4.0587>
   rotation=<rot -9.5026 -0.6723 -90.4748>
  }]
 dyn_jacket_bk_04_r [{
   shapeType=Capsule
   length=9.3516
   radius=7.7992
   shapeName=dyn_jacket_fr_04_r_capsule
   center=<6.3625 0.4418 -3.1639>
   rotation=<rot 0.0755 0.3171 -98.0605>
  }]
 dyn_jacket_fr_01_l [{
   shapeType=Capsule
   length=3.8622
   radius=3.8417
   shapeName=dyn_jacket_fr_01_l_capsule
   center=<5.6068 0.2145 0.5137>
   rotation=<rot -0.383 90.6332 -88.1665>
  }]
 dyn_jacket_fr_01_r [{
   shapeType=Capsule
   length=7.7622
   radius=3.8417
   shapeName=dyn_jacket_fr_01_l_capsule
   center=<6.3577 0.223 0.5377>
   rotation=<rot -0.383 90.6332 -88.1665>
  }]
 dyn_jacket_fr_02_l [{
   shapeType=Capsule
   length=5.8
   radius=4.505
   shapeName=dyn_jacket_fr_02_l_capsule
   center=<4.1939 0 0.4471>
   rotation=<rot -3.0597 25.4489 90.0944>
  }]
 dyn_jacket_fr_02_r [{
   shapeType=Capsule
   length=6.1
   radius=4.205
   shapeName=dyn_jacket_fr_02_l_capsule
   center=<4.2123 0.1085 0.7911>
   rotation=<rot 2.8692 -22.7631 97.2029>
  }]
 dyn_jacket_fr_03_l [{
   shapeType=Capsule
   length=5.2138
   radius=4.9105
   shapeName=dyn_jacket_fr_03_l_capsule
   center=<3.3836 -0.2761 0.1613>
   rotation=<rot -1.1923 -0.4573 -86.3944>
  }]
 dyn_jacket_fr_03_r [{
   shapeType=Capsule
   length=6.3516
   radius=4.6992
   shapeName=dyn_jacket_fr_04_r_capsule
   center=<4.0862 1.5108 0.7685>
   rotation=<rot 0.1097 0.1888 -84.9151>
  }]
 dyn_jacket_fr_04_l [{
   shapeType=Capsule
   length=6.7828
   radius=5.764
   shapeName=dyn_jacket_fr_04_l_capsule
   center=<3.5683 -0.3544 -0.1635>
   rotation=<rot 5.0705 179.0394 -89.8871>
  }]
 dyn_jacket_fr_04_r [{
   shapeType=Capsule
   length=10.2516
   radius=4.5992
   shapeName=dyn_jacket_fr_04_r_capsule
   center=<5.5128 -0.0489 0.7994>
   rotation=<rot -3.0447 12.5906 -88.3261>
  }]
 dyn_jacket_mid_01 [{
   shapeType=Capsule
   length=13.9989
   radius=7.1
   shapeName=dyn_jacket_mid_03_capsule
   center=<5.3663 0.3199 -3.2144>
   rotation=<rot 0.0095 -0.5172 -89.5787>
  }]
 dyn_jacket_mid_02 [{
   shapeType=Capsule
   length=13.9989
   radius=7.1
   shapeName=dyn_jacket_mid_03_capsule
   center=<5.3662 0.3275 -2.1804>
   rotation=<rot 0.0095 -0.5172 -89.5787>
  }]
 dyn_jacket_mid_03 [{
   shapeType=Capsule
   length=13.9989
   radius=7.1
   shapeName=dyn_jacket_mid_03_capsule
   center=<5.3661 0.3297 -1.8906>
   rotation=<rot 0.0095 -0.5172 -89.5787>
  }]
 dyn_jacket_mid_04 [{
   shapeType=Capsule
   length=13.9989
   radius=7.1
   shapeName=dyn_jacket_mid_03_capsule
   center=<5.0682 0.3338 -1.6911>
   rotation=<rot 0.0032 0.3309 -89.5786>
  }]
 dyn_jacket_side_01_l [{
   shapeType=Capsule
   length=8.7516
   radius=3.4992
   shapeName=dyn_jacket_fr_04_r_capsule
   center=<6.4979 0.2124 1.4743>
   rotation=<rot -0.388 89.8471 -84.3474>
  }]
 dyn_jacket_side_01_r [{
   shapeType=Capsule
   length=12.1516
   radius=3.4992
   shapeName=dyn_jacket_fr_04_r_capsule
   center=<9.0398 -0.2237 3.1461>
   rotation=<rot -0.5317 -0.1602 90.3882>
  }]
 dyn_jacket_side_02_l [{
   shapeType=Capsule
   length=8.2516
   radius=3.8992
   shapeName=dyn_jacket_fr_04_r_capsule
   center=<4.9617 0.2104 1.5475>
   rotation=<rot -0.3819 90.5118 -87.4098>
  }]
 dyn_jacket_side_02_r [{
   shapeType=Capsule
   length=11.0516
   radius=4.8992
   shapeName=dyn_jacket_fr_04_r_capsule
   center=<5.0489 -0.2033 3.1858>
   rotation=<rot 1.3434 -0.1457 -89.5144>
  }]
 dyn_jacket_side_03_l [{
   shapeType=Capsule
   length=8.9516
   radius=4.6992
   shapeName=dyn_jacket_fr_04_r_capsule
   center=<6.9508 0.0712 1.9095>
   rotation=<rot -3.8532 91.1328 -91.2026>
  }]
 dyn_jacket_side_03_r [{
   shapeType=Capsule
   length=11.5516
   radius=5.2992
   shapeName=dyn_jacket_fr_04_r_capsule
   center=<5.8236 -0.0507 3.3116>
   rotation=<rot -0.284 -0.1354 -91.7179>
  }]
 dyn_jacket_side_04_l [{
   shapeType=Capsule
   length=4.9516
   radius=5.7992
   shapeName=dyn_jacket_fr_04_r_capsule
   center=<3.9993 -0.0689 -0.7278>
   rotation=<rot -5.7539 -0.2805 -89.9544>
  }]
 dyn_jacket_side_04_r [{
   shapeType=Capsule
   length=13.1516
   radius=5.3992
   shapeName=dyn_jacket_fr_04_r_capsule
   center=<4.7989 0.262 1.6442>
   rotation=<rot -0.1047 1.8368 -87.0005>
  }]
 extra_ball_l [{
   shapeType=Capsule
   length=3.9525
   radius=5.5217
   shapeName=extra_ball_r_capsule
   center=<-1.7222 3.6445 0.0192>
   rotation=<rot 0.6718 -179.03 94.6138>
  }]
 extra_ball_r [{
   shapeType=Capsule
   length=3.9525
   radius=5.5217
   shapeName=extra_ball_r_capsule
   center=<1.7271 -3.6627 0.2071>
   rotation=<rot -1.1648 1.0099 85.3855>
  }]
 extra_foot_l [{
   shapeType=Capsule
   length=6.9622
   radius=7.0512
   shapeName=extra_foot_l_capsule
   center=<6.0402 -0.1888 -0.6113>
   rotation=<rot -0.0062 -94.3256 -88.2731>
  }]
 extra_foot_r [{
   shapeType=Capsule
   length=7.2306
   radius=6.8913
   shapeName=extra_foot_r_capsule
   center=<6.4927 -0.5406 0.5948>
   rotation=<rot -0.0876 85.5653 -88.242>
  }]
 pelvis [{shapeType=Capsule, length=12.4622, radius=12.0827, shapeName=pelvis_capsule, center=<-2.019 -1.7783 0.1303>}]
 root [{shapeType=Sphere, radius=1202.7058, shapeName=root_sphere, center=<-0.2446 108.0086 -1193.971>}]
 spine_01 [
  {shapeType=Capsule, length=22.9364, radius=10.4425, shapeName=spine_01_capsule, center=<-2.9029 -7.5061 -0.0246>, rotation=<rot 0 0 -0.0714>}
  {shapeType=Capsule, length=14.7, radius=10, shapeName=spine_01_capsule, center=<-4.9447 -0.6566 0>, rotation=<rot 0 0 -0.0714>}
  {
   shapeType=Capsule
   length=24.5
   radius=5.5
   shapeName=spine_01_capsule
   center=<-0.9605 16.6568 0.2405>
   rotation=<rot -89.5042 -179.6449 180>
  }
 ]
 spine_02 [
  {
   shapeType=Capsule
   length=20.4364
   radius=10.4425
   shapeName=spine_01_capsule
   center=<-1.4782 -7.7367 -0.0249>
   rotation=<rot -0.0059 -175.2473 0.0712>
  }
  {shapeType=Capsule, length=14.7, radius=10, shapeName=spine_01_capsule, center=<-4.9447 -0.6566 0>, rotation=<rot 0 0 -0.0714>}
  {
   shapeType=Capsule
   length=24.5
   radius=5.5
   shapeName=spine_01_capsule
   center=<-0.9605 16.6568 0.2405>
   rotation=<rot -89.5042 -179.6449 180>
  }
 ]
 spine_03 [{
   shapeType=Capsule
   length=13.4601
   radius=14.3769
   shapeName=spine_03_capsule
   center=<-0.2374 -1.5305 0.0084>
   rotation=<rot 0 3.1923 -1.2208>
  }]
 spine_04 [{
   shapeType=Capsule
   length=14.315
   radius=13.4939
   shapeName=spine_04_capsule
   center=<3.7265 -0.3862 0.1726>
   rotation=<rot -0.0522 -90 0.5439>
  }]
 thigh_l [
  {
   shapeType=Capsule
   length=20.5582
   radius=8.2011
   shapeName=thigh_l_capsule
   center=<-28.3273 -2.1461 -0.1702>
   rotation=<rot 90 171.1193 167.5007>
  }
  {shapeType=Capsule, length=10.2, radius=9.5, shapeName=thigh_l_capsule, center=<-12.4942 0 8.2619>, rotation=<rot 87.8291 0 0.0001>}
 ]
 thigh_r [{shapeType=Capsule, length=34.6683, radius=8.746, shapeName=thigh_r_capsule, center=<20.4052 1.0735 0.2584>, rotation=<rot 90 180 177.586>}]
}
constraints [
 {bone1Name=dyn_jacket_mid_02, bone2Name=dyn_jacket_mid_01, swing1LimitDegrees=7, swing2LimitDegrees=45, twistLimitDegrees=0}
 {bone1Name=dyn_jacket_mid_03, bone2Name=dyn_jacket_mid_02, swing1LimitDegrees=20, swing2LimitDegrees=45, twistLimitDegrees=0}
 {bone1Name=dyn_jacket_mid_04, bone2Name=dyn_jacket_mid_03, swing1LimitDegrees=15, swing2LimitDegrees=45, twistLimitDegrees=5}
 {bone1Name=dyn_jacket_mid_01, bone2Name=dyn_jacket_base_mid, swing1LimitDegrees=2, swing2LimitDegrees=20, twistLimitDegrees=0}
 {bone1Name=dyn_jacket_base_mid, bone2Name=spine_03, swing1LimitDegrees=10, swing2LimitDegrees=45, twistLimitDegrees=5}
 {bone1Name=dyn_jacket_side_03_r, bone2Name=dyn_jacket_side_02_r, swing1LimitDegrees=10, swing2LimitDegrees=45, twistLimitDegrees=5}
 {bone1Name=dyn_jacket_side_04_r, bone2Name=dyn_jacket_side_03_r, swing1LimitDegrees=10, swing2LimitDegrees=45, twistLimitDegrees=5}
 {bone1Name=dyn_jacket_fr_04_r, bone2Name=dyn_jacket_fr_03_r, swing1LimitDegrees=5, swing2LimitDegrees=55, twistLimitDegrees=20}
 {bone1Name=dyn_jacket_fr_03_r, bone2Name=dyn_jacket_fr_02_r, swing1LimitDegrees=35, swing2LimitDegrees=55, twistLimitDegrees=25}
 {bone1Name=dyn_jacket_bk_04_r, bone2Name=dyn_jacket_bk_03_r, swing1LimitDegrees=15, swing2LimitDegrees=15, twistLimitDegrees=5}
 {bone1Name=dyn_jacket_bk_03_r, bone2Name=dyn_jacket_bk_02_r, swing1LimitDegrees=15, swing2LimitDegrees=25, twistLimitDegrees=5}
 {
  bone1Name=spine_02
  bone2Name=spine_01
  swing1LimitDegrees=45
  swing1Motion=Free
  swing2LimitDegrees=45
  swing2Motion=Free
  twistLimitDegrees=45
  twistMotion=Free
 }
 {
  bone1Name=spine_01
  bone2Name=pelvis
  swing1LimitDegrees=45
  swing1Motion=Free
  swing2LimitDegrees=45
  swing2Motion=Free
  twistLimitDegrees=45
  twistMotion=Free
 }
 {
  bone1Name=thigh_l
  bone2Name=pelvis
  swing1LimitDegrees=45
  swing1Motion=Free
  swing2LimitDegrees=45
  swing2Motion=Free
  twistLimitDegrees=45
  twistMotion=Free
 }
 {
  bone1Name=thigh_r
  bone2Name=pelvis
  swing1LimitDegrees=45
  swing1Motion=Free
  swing2LimitDegrees=45
  swing2Motion=Free
  twistLimitDegrees=45
  twistMotion=Free
 }
 {bone1Name=dyn_jacket_fr_04_l, bone2Name=dyn_jacket_fr_03_l, swing1LimitDegrees=5, swing2LimitDegrees=65, twistLimitDegrees=15}
 {bone1Name=dyn_jacket_base_fr_l, bone2Name=spine_03, swing1LimitDegrees=10, swing2LimitDegrees=45, twistLimitDegrees=5}
 {bone1Name=dyn_jacket_fr_01_l, bone2Name=dyn_jacket_base_fr_l, swing1LimitDegrees=10, swing2LimitDegrees=35, twistLimitDegrees=25}
 {bone1Name=dyn_jacket_fr_02_l, bone2Name=dyn_jacket_fr_01_l, swing1LimitDegrees=15, swing2LimitDegrees=65, twistLimitDegrees=35}
 {bone1Name=dyn_jacket_fr_03_l, bone2Name=dyn_jacket_fr_02_l, swing1LimitDegrees=35, swing2LimitDegrees=55, twistLimitDegrees=15}
 {
  bone1Name=dyn_jacket_bk_04_l
  bone2Name=dyn_jacket_bk_03_l
  swing1LimitDegrees=10
  swing1Motion=Locked
  swing2LimitDegrees=45
  twistLimitDegrees=5
 }
 {bone1Name=dyn_jacket_bk_02_l, bone2Name=dyn_jacket_bk_01_l, swing1LimitDegrees=15, swing2LimitDegrees=25, twistLimitDegrees=5}
 {bone1Name=dyn_jacket_bk_03_l, bone2Name=dyn_jacket_bk_02_l, swing1LimitDegrees=15, swing2LimitDegrees=25, twistLimitDegrees=5}
 {bone1Name=dyn_jacket_bk_01_l, bone2Name=dyn_jacket_base_bk_l, swing1LimitDegrees=15, swing2LimitDegrees=40, twistLimitDegrees=5}
 {bone1Name=dyn_jacket_base_bk_l, bone2Name=spine_03, swing1LimitDegrees=10, swing2LimitDegrees=45, twistLimitDegrees=5}
 {bone1Name=dyn_jacket_side_04_l, bone2Name=dyn_jacket_side_03_l, swing1LimitDegrees=5, swing2LimitDegrees=45, twistLimitDegrees=45}
 {bone1Name=dyn_jacket_side_03_l, bone2Name=dyn_jacket_side_02_l, swing1LimitDegrees=10, swing2LimitDegrees=45, twistLimitDegrees=45}
 {bone1Name=dyn_jacket_side_02_l, bone2Name=dyn_jacket_side_01_l, swing1LimitDegrees=10, swing2LimitDegrees=45, twistLimitDegrees=15}
 {bone1Name=dyn_jacket_side_01_l, bone2Name=dyn_jacket_base_side_l, swing1LimitDegrees=10, swing2LimitDegrees=45, twistLimitDegrees=10}
 {bone1Name=dyn_jacket_base_side_l, bone2Name=spine_03, swing1LimitDegrees=10, swing2LimitDegrees=45, twistLimitDegrees=5}
 {bone1Name=dyn_jacket_side_03_l, bone2Name=dyn_jacket_fr_03_l, swing1LimitDegrees=45, swing2LimitDegrees=45, twistLimitDegrees=45}
 {bone1Name=dyn_jacket_side_04_r, bone2Name=dyn_jacket_fr_04_r, swing1LimitDegrees=45, swing2LimitDegrees=45, twistLimitDegrees=45}
 {
  bone1Name=dyn_jacket_mid_03
  bone2Name=dyn_jacket_bk_03_l
  swing1LimitDegrees=45
  swing1Motion=Locked
  swing2LimitDegrees=45
  swing2Motion=Free
  twistLimitDegrees=45
  twistMotion=Free
 }
 {bone1Name=dyn_jacket_bk_03_l, bone2Name=dyn_jacket_side_03_l, swing1LimitDegrees=5, swing2LimitDegrees=5, twistLimitDegrees=45}
 {bone1Name=dyn_jacket_fr_02_r, bone2Name=dyn_jacket_fr_01_r, swing1LimitDegrees=75, swing2LimitDegrees=45, twistLimitDegrees=20}
 {bone1Name=dyn_jacket_fr_01_r, bone2Name=dyn_jacket_base_fr_r, swing1LimitDegrees=20, swing2LimitDegrees=45, twistLimitDegrees=35}
 {bone1Name=dyn_jacket_bk_02_r, bone2Name=dyn_jacket_bk_01_r, swing1LimitDegrees=15, swing2LimitDegrees=25, twistLimitDegrees=5}
 {bone1Name=dyn_jacket_bk_01_r, bone2Name=dyn_jacket_base_bk_r, swing1LimitDegrees=15, swing2LimitDegrees=40, twistLimitDegrees=5}
 {bone1Name=dyn_jacket_side_02_r, bone2Name=dyn_jacket_side_01_r, swing1LimitDegrees=10, swing2LimitDegrees=45, twistLimitDegrees=5}
 {bone1Name=dyn_jacket_side_01_r, bone2Name=dyn_jacket_base_side_r, swing1LimitDegrees=10, swing2LimitDegrees=45, twistLimitDegrees=5}
 {
  bone1Name=dyn_jacket_base_bk_r
  bone2Name=spine_03
  swing1LimitDegrees=45
  swing1Motion=Free
  swing2LimitDegrees=45
  swing2Motion=Free
  twistLimitDegrees=45
  twistMotion=Free
 }
 {
  bone1Name=dyn_jacket_base_fr_r
  bone2Name=spine_03
  swing1LimitDegrees=45
  swing1Motion=Free
  swing2LimitDegrees=45
  swing2Motion=Free
  twistLimitDegrees=45
  twistMotion=Free
 }
 {
  bone1Name=dyn_jacket_base_side_r
  bone2Name=spine_03
  swing1LimitDegrees=45
  swing1Motion=Free
  swing2LimitDegrees=45
  swing2Motion=Free
  twistLimitDegrees=45
  twistMotion=Free
 }
 {bone1Name=dyn_jacket_bk_04_r, bone2Name=dyn_jacket_mid_04, swing1LimitDegrees=5, swing2LimitDegrees=5, twistLimitDegrees=5}
 {bone1Name=dyn_jacket_bk_04_l, bone2Name=dyn_jacket_side_04_l, swing1LimitDegrees=5, swing2LimitDegrees=5, twistLimitDegrees=15}
 {bone1Name=dyn_jacket_side_04_l, bone2Name=dyn_jacket_fr_04_l, swing1LimitDegrees=45, swing2LimitDegrees=45, twistLimitDegrees=45}
 {bone1Name=dyn_jacket_mid_04, bone2Name=dyn_jacket_bk_04_l, swing1LimitDegrees=5, swing2LimitDegrees=5, twistLimitDegrees=5}
 {
  bone1Name=pelvis
  bone2Name=root
  swing1LimitDegrees=45
  swing1Motion=Free
  swing2LimitDegrees=45
  swing2Motion=Free
  twistLimitDegrees=45
  twistMotion=Free
 }
 {bone1Name=dyn_jacket_side_03_r, bone2Name=dyn_jacket_bk_03_r, swing1LimitDegrees=15, swing2LimitDegrees=45, twistLimitDegrees=45}
 {bone1Name=dyn_jacket_side_04_r, bone2Name=dyn_jacket_bk_04_r, swing1LimitDegrees=45, swing2LimitDegrees=45, twistLimitDegrees=45}
 {
  bone1Name=spine_03
  bone2Name=spine_02
  swing1LimitDegrees=45
  swing1Motion=Free
  swing2LimitDegrees=45
  swing2Motion=Free
  twistLimitDegrees=45
  twistMotion=Free
 }
 {bone1Name=dyn_jacket_bk_01_l, bone2Name=dyn_jacket_side_01_l, swing1LimitDegrees=45, swing2LimitDegrees=45, twistLimitDegrees=45}
 {
  bone1Name=dyn_jacket_side_01_l
  bone2Name=dyn_jacket_fr_01_l
  swing1LimitDegrees=5
  swing2LimitDegrees=5
  twistLimitDegrees=5
  twistMotion=Free
 }
 {
  bone1Name=calf_r
  bone2Name=thigh_r
  swing1LimitDegrees=45
  swing1Motion=Free
  swing2LimitDegrees=45
  swing2Motion=Free
  twistLimitDegrees=45
  twistMotion=Free
 }
 {
  bone1Name=calf_l
  bone2Name=thigh_l
  swing1LimitDegrees=45
  swing1Motion=Free
  swing2LimitDegrees=45
  swing2Motion=Free
  twistLimitDegrees=45
  twistMotion=Free
 }
 {
  bone1Name=dyn_jacket_side_03_r
  bone2Name=dyn_jacket_fr_03_r
  swing1LimitDegrees=45
  swing1Motion=Free
  swing2LimitDegrees=45
  swing2Motion=Free
  twistLimitDegrees=45
  twistMotion=Free
 }
 {
  bone1Name=extra_foot_r
  bone2Name=calf_r
  swing1LimitDegrees=45
  swing1Motion=Free
  swing2LimitDegrees=45
  swing2Motion=Free
  twistLimitDegrees=45
  twistMotion=Free
 }
 {
  bone1Name=extra_foot_l
  bone2Name=calf_l
  swing1LimitDegrees=45
  swing1Motion=Free
  swing2LimitDegrees=45
  swing2Motion=Free
  twistLimitDegrees=45
  twistMotion=Free
 }
 {
  bone1Name=extra_ball_r
  bone2Name=extra_foot_r
  swing1LimitDegrees=45
  swing1Motion=Free
  swing2LimitDegrees=45
  swing2Motion=Free
  twistLimitDegrees=45
  twistMotion=Free
 }
 {
  bone1Name=extra_ball_l
  bone2Name=extra_foot_l
  swing1LimitDegrees=45
  swing1Motion=Free
  swing2LimitDegrees=45
  swing2Motion=Free
  twistLimitDegrees=45
  twistMotion=Free
 }
 {
  bone1Name=spine_04
  bone2Name=spine_03
  swing1LimitDegrees=45
  swing1Motion=Free
  swing2LimitDegrees=45
  swing2Motion=Free
  twistLimitDegrees=45
  twistMotion=Free
 }
]
