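# Physics asset for the FaceAcc_BackBling mesh part of the female Renegade Whip character body, defining physics for its hair bones.
# Side chains hang from the head, while paired 04 chains and a back chain branch from a back base bone, all with limited swing and twist.
# Capsules on head, upper spine, clavicles and upper arms give the hair body volumes, joined by free-rotation constraints, as are the hair anchor links.
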PhysicsAsset /BRCosmetics/Characters/Player/Female/Medium/Bodies/F_MED_RenegadeWhip/Meshes/Parts/F_MED_RenegadeWhip_FaceAcc_BackBling_Physics
bodies {
 clavicle_l [{
   shapeType=Capsule
   length=14.2
   radius=7.605
   shapeName=clavicle_r_capsule
   center=<10.6063 2.1103 -3.082>
   rotation=<rot -71.2713 -178.2649 13.6972>
  }]
 clavicle_r [{
   shapeType=Capsule
   length=10.1
   radius=7.605
   shapeName=clavicle_r_capsule
   center=<-9.127 -1.6006 1.98>
   rotation=<rot 71.044 1.7286 -16.5051>
  }]
 dyn_hair_01_01_l [{
   shapeType=Capsule
   length=4.235
   radius=1.7614
   shapeName=dyn_hair_01_01_l_capsule
   center=<3.0332 -0.1882 0.0309>
   rotation=<rot -1.6385 77.5201 81.3416>
  }]
 dyn_hair_01_01_r [{
   shapeType=Capsule
   length=3.8766
   radius=1.5532
   shapeName=dyn_hair_01_01_r_capsule
   center=<3.266 -0.0133 -0.0925>
   rotation=<rot -0.0031 87.3003 85.5617>
  }]
 dyn_hair_01_02_l [{
   shapeType=Capsule
   length=4.1535
   radius=1.5396
   shapeName=dyn_hair_01_02_l_capsule
   center=<2.7427 -0.1821 -0.1259>
   rotation=<rot 0.1264 86.6979 85.7612>
  }]
 dyn_hair_01_02_r [{
   shapeType=Capsule
   length=4.1387
   radius=1.4044
   shapeName=dyn_hair_01_02_r_capsule
   center=<3.5854 -0.2427 0.4798>
   rotation=<rot 0 81.3173 -78.2937>
  }]
 dyn_hair_01_03_l [{
   shapeType=Capsule
   length=4.2641
   radius=1.8344
   shapeName=dyn_hair_01_03_l_capsule
   center=<3.0674 -0.1795 0.0328>
   rotation=<rot 0 83.2541 -89.757>
  }]
 dyn_hair_01_03_r [{
   shapeType=Capsule
   length=3.0359
   radius=1.5518
   shapeName=dyn_hair_01_03_r_capsule
   center=<2.974 -0.4226 0.0001>
   rotation=<rot -0.0031 76.3464 88.2417>
  }]
 dyn_hair_04_01_l [{
   shapeType=Capsule
   length=3.9936
   radius=2.6138
   shapeName=dyn_hair_04_01_l_capsule
   center=<3.5341 -1.7443 0.7978>
   rotation=<rot 85.6141 5.0014 22.9968>
  }]
 dyn_hair_04_01_r [{
   shapeType=Capsule
   length=4.1627
   radius=2.3934
   shapeName=dyn_hair_04_01_r_capsule
   center=<3.6388 -0.496 0.3085>
   rotation=<rot 0.0045 83.3907 -90.0341>
  }]
 dyn_hair_04_02_l [{
   shapeType=Capsule
   length=1.8224
   radius=2.7243
   shapeName=dyn_hair_04_02_l_capsule
   center=<2.3167 -0.0743 -0.1603>
   rotation=<rot 23.8195 87.497 79.592>
  }]
 dyn_hair_04_02_r [{
   shapeType=Capsule
   length=3.3062
   radius=2.229
   shapeName=dyn_hair_04_02_r_capsule
   center=<3.8107 -0.691 0.0563>
   rotation=<rot -1.8406 84.7412 -88.63>
  }]
 dyn_hair_04_03_l [{
   shapeType=Capsule
   length=6.4479
   radius=2.3684
   shapeName=dyn_hair_04_03_l_capsule
   center=<3.7792 -0.7569 0.5683>
   rotation=<rot 0 87.061 -82.5297>
  }]
 dyn_hair_04_03_r [{
   shapeType=Capsule
   length=6.4062
   radius=2.329
   shapeName=dyn_hair_04_02_r_capsule
   center=<4.1935 -0.8665 0.3155>
   rotation=<rot -1.7239 96.0021 -86.0843>
  }]
 dyn_hair_04_base_l [{shapeType=Sphere, radius=0.5, shapeName=dyn_hair_04_base_l_sphere}]
 dyn_hair_04_base_r [{shapeType=Sphere, radius=0.5, shapeName=dyn_hair_04_base_r_sphere}]
 dyn_hair_back_01 [{
   shapeType=Capsule
   length=3.7125
   radius=2.6051
   shapeName=dyn_hair_back_02_capsule
   center=<7.3654 -0.4289 -1.933>
   rotation=<rot 0 -88.2029 -3.4365>
  }]
 dyn_hair_back_02 [{
   shapeType=Capsule
   length=4.1048
   radius=2.6352
   shapeName=dyn_hair_back_02_capsule
   center=<6.6416 -0.1265 0.6408>
   rotation=<rot 0 -88.203 -15.1137>
  }]
 dyn_hair_back_MSTR [{
   shapeType=Capsule
   length=0.8454
   radius=1.2401
   shapeName=dyn_hair_back_MSTR_capsule
   center=<0.7391 0 -1.4065>
   rotation=<rot 0 -89.9999 62.2793>
  }]
 dyn_hair_base_back [{shapeType=Capsule, length=0.1, radius=0.5, shapeName=dyn_hair_base_back_capsule, center=<6.8952 0.1427 -0.2826>}]
 head [
  {
   shapeType=Capsule
   length=8.6114
   radius=5.9537
   shapeName=head_capsule
   center=<1.1353 1.0556 0.1993>
   rotation=<rot 0.1129 -81.5316 -89.8304>
  }
  {shapeType=Capsule, length=7.9591, radius=5.5, shapeName=head_capsule, center=<0.2923 7.3059 -0.0028>, rotation=<rot 89.445 -0.0001 0>}
 ]
 spine_04 [{shapeType=Capsule, length=41.5, radius=13.7, shapeName=spine_04_capsule, center=<3.7866 9.6361 0>}]
 spine_05 [
  {shapeType=Capsule, length=39.3, radius=9.505, shapeName=spine_05_capsule, center=<3.6143 -0.4264 0>, rotation=<rot 0 -174.7973 0>}
  {shapeType=Capsule, length=30.5, radius=15, shapeName=spine_05_capsule, center=<14.5486 -29.1209 0>}
 ]
 upperarm_l [{
   shapeType=Capsule
   length=23.1
   radius=5.005
   shapeName=upperarm_l_capsule
   center=<17.1429 0.185 -0.2298>
   rotation=<rot -89.586 -137.2138 137.2241>
  }]
 upperarm_r [{
   shapeType=Capsule
   length=22.9685
   radius=5.1365
   shapeName=upperarm_l_capsule
   center=<-18.4655 0.1792 -0.0409>
   rotation=<rot 89.5785 43.9 -136.1165>
  }]
}
constraints [
 {bone1Name=dyn_hair_04_02_r, bone2Name=dyn_hair_04_01_r, swing1LimitDegrees=35, swing2LimitDegrees=15, twistLimitDegrees=10}
 {bone1Name=dyn_hair_04_01_r, bone2Name=dyn_hair_04_base_r, swing1LimitDegrees=18, swing2LimitDegrees=5, twistLimitDegrees=5}
 {bone1Name=dyn_hair_04_03_l, bone2Name=dyn_hair_04_02_l, swing1LimitDegrees=25, swing2LimitDegrees=15, twistLimitDegrees=10}
 {bone1Name=dyn_hair_04_02_l, bone2Name=dyn_hair_04_01_l, swing1LimitDegrees=10, swing2LimitDegrees=10, twistLimitDegrees=10}
 {bone1Name=dyn_hair_01_03_r, bone2Name=dyn_hair_01_02_r, swing1LimitDegrees=45, swing2LimitDegrees=45, twistLimitDegrees=5}
 {bone1Name=dyn_hair_01_02_r, bone2Name=dyn_hair_01_01_r, swing1LimitDegrees=45, swing2LimitDegrees=45, twistLimitDegrees=5}
 {bone1Name=dyn_hair_01_01_r, bone2Name=head, swing1LimitDegrees=45, swing2LimitDegrees=45, twistLimitDegrees=5}
 {bone1Name=dyn_hair_01_03_l, bone2Name=dyn_hair_01_02_l, swing1LimitDegrees=45, swing2LimitDegrees=45, twistLimitDegrees=5}
 {bone1Name=dyn_hair_01_02_l, bone2Name=dyn_hair_01_01_l, swing1LimitDegrees=45, swing2LimitDegrees=45, twistLimitDegrees=5}
 {bone1Name=dyn_hair_01_01_l, bone2Name=head, swing1LimitDegrees=45, swing2LimitDegrees=45, twistLimitDegrees=5}
 {
  bone1Name=head
  bone2Name=spine_05
  swing1LimitDegrees=45
  swing1Motion=Free
  swing2LimitDegrees=45
  swing2Motion=Free
  twistLimitDegrees=45
  twistMotion=Free
 }
 {
  bone1Name=clavicle_l
  bone2Name=spine_05
  swing1LimitDegrees=45
  swing1Motion=Free
  swing2LimitDegrees=45
  swing2Motion=Free
  twistLimitDegrees=45
  twistMotion=Free
 }
 {
  bone1Name=upperarm_l
  bone2Name=clavicle_l
  swing1LimitDegrees=45
  swing1Motion=Free
  swing2LimitDegrees=45
  swing2Motion=Free
  twistLimitDegrees=45
  twistMotion=Free
 }
 {
  bone1Name=clavicle_r
  bone2Name=spine_05
  swing1LimitDegrees=45
  swing1Motion=Free
  swing2LimitDegrees=45
  swing2Motion=Free
  twistLimitDegrees=45
  twistMotion=Free
 }
 {
  bone1Name=upperarm_r
  bone2Name=clavicle_r
  swing1LimitDegrees=45
  swing1Motion=Free
  swing2LimitDegrees=45
  swing2Motion=Free
  twistLimitDegrees=45
  twistMotion=Free
 }
 {
  bone1Name=dyn_hair_base_back
  bone2Name=head
  swing1LimitDegrees=45
  swing1Motion=Free
  swing2LimitDegrees=45
  swing2Motion=Free
  twistLimitDegrees=45
  twistMotion=Free
 }
 {bone1Name=dyn_hair_04_03_r, bone2Name=dyn_hair_04_02_r, swing1LimitDegrees=25, swing2LimitDegrees=15, twistLimitDegrees=10}
 {
  bone1Name=spine_05
  bone2Name=spine_04
  swing1LimitDegrees=45
  swing1Motion=Free
  swing2LimitDegrees=45
  swing2Motion=Free
  twistLimitDegrees=45
  twistMotion=Free
 }
 {bone1Name=dyn_hair_04_01_l, bone2Name=dyn_hair_04_base_l, swing1LimitDegrees=25, swing2LimitDegrees=25, twistLimitDegrees=10}
 {
  bone1Name=dyn_hair_04_base_l
  bone2Name=dyn_hair_base_back
  swing1LimitDegrees=45
  swing1Motion=Free
  swing2LimitDegrees=45
  swing2Motion=Free
  twistLimitDegrees=45
  twistMotion=Free
 }
 {
  bone1Name=dyn_hair_04_base_r
  bone2Name=dyn_hair_base_back
  swing1LimitDegrees=45
  swing1Motion=Free
  swing2LimitDegrees=45
  swing2Motion=Free
  twistLimitDegrees=45
  twistMotion=Free
 }
 {
  bone1Name=dyn_hair_back_MSTR
  bone2Name=dyn_hair_base_back
  swing1LimitDegrees=45
  swing1Motion=Free
  swing2LimitDegrees=45
  swing2Motion=Free
  twistLimitDegrees=45
  twistMotion=Free
 }
 {bone1Name=dyn_hair_back_01, bone2Name=dyn_hair_back_MSTR, swing1LimitDegrees=35, swing2LimitDegrees=20, twistLimitDegrees=10}
 {bone1Name=dyn_hair_back_02, bone2Name=dyn_hair_back_01, swing1LimitDegrees=25, swing2LimitDegrees=15, twistLimitDegrees=5}
]
