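# Control Rig for the RenegadeWhip female face accessory mesh, applying corrective offsets to hair bones as the head rotates.
# It applies fixed hair-root adjustments, then a sequence branch reads the head's delta through the Get Delta Ref function.
# Remapped negative yaw weights a back-hair base offset and gates further pitch-driven correctives.
# Positive pitch adjusts the back base and a right strand base, while negative pitch adjusts only the back base.
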ControlRigBlueprint /BRCosmetics/Characters/Player/Female/Medium/Bodies/F_MED_RenegadeWhip/Meshes/Parts/F_MED_RenegadeWhip_FaceAcc_CtrlRig
graphs [
 RigVMFunctionLibrary {links [], nodes []}
 RigVMFunction_Sequence_2_ContainedGraph {nodes [
   n1 = RigVMFunctionEntryNode()
   RigVMFunction_Sequence = &RigVMFunction_Sequence(ExecuteContext=@n1.ExecuteContext)
   n2 = &RigVMFunction_Sequence(ExecuteContext=@RigVMFunction_Sequence.B)
   n3 = &RigVMFunction_Sequence(ExecuteContext=@n2.B)
   RigVMFunctionReturnNode(A=@RigVMFunction_Sequence.A, B=@n2.A, C=@n3.A, D=@n3.B)
  ]}
 RigVMModel {nodes [
   RigUnit_BeginExecution = &RigUnit_BeginExecution()
   &RigUnit_ModifyTransforms(
    ItemToModify=((Item=(Type=Bone,Name="deform_hair_root_r"),Transform=(Rotation=(X=0,Y=0,Z=0,W=1),Translation=(X=0,Y=-2.5,Z=0),Scale3D=(X=1,Y=1,Z=1))))
    Weight=1
    WeightMinimum=0
    ExecutePin=@RigUnit_BeginExecution.ExecutePin
   )
   > &RigUnit_ModifyTransforms(
    ItemToModify=((Item=(Type=Bone,Name="deform_hair_root_l"),Transform=(Rotation=(X=0,Y=0,Z=0,W=1),Translation=(X=0,Y=0,Z=0),Scale3D=(X=1,Y=1,Z=1))))
    Weight=1
    WeightMinimum=0
   )
   > RigVMFunction_Sequence = RigVMAggregateNode()
   &RigUnit_ModifyTransforms(
    ItemToModify=((Item=(Type=Bone,Name="head"),Transform=(Rotation=(X=0,Y=0,Z=0,W=1),Translation=(X=0,Y=0,Z=0),Scale3D=(X=1,Y=1,Z=1))))
    Weight=1
    WeightMinimum=0
    ExecutePin=@RigVMFunction_Sequence.B
   )
   > Get Delta Ref = RigVMFunctionReferenceNode(Item=(Type=Bone,Name="head"))
   Remap = &RigVMFunction_MathDoubleRemap(SourceMaximum=-30, SourceMinimum=-8, TargetMaximum=1, TargetMinimum=0, Value=@"Get Delta Ref".Result.Rotation.Yaw)
   > &RigUnit_ModifyTransforms(
    ItemToModify=((Item=(Type=Bone,Name="dyn_hair_base_back"),Transform=(Rotation=(X=0,Y=0,Z=-0.5373,W=0.8434),Translation=(X=-4,Y=0,Z=2),Scale3D=(X=1,Y=1,Z=1.2))))
    WeightMinimum=0
    Weight=@Remap.Result
   )
   Multiply = &RigVMFunction_MathDoubleMul(
    A=@Remap.Result
    B=&RigVMFunction_MathDoubleRemap(SourceMaximum=15, SourceMinimum=0, TargetMaximum=1, TargetMinimum=0, Value=@"Get Delta Ref".Result.Rotation.Pitch).Result
   )
   > &RigUnit_ModifyTransforms(
    ItemToModify=((Item=(Type=Bone,Name="dyn_hair_base_back"),Transform=(Rotation=(X=-0.2988,Y=0.0521,Z=-0.0201,W=0.9527),Translation=(X=1.5,Y=0,Z=1),Scale3D=(X=1,Y=1,Z=1))))
    WeightMinimum=0
    Weight=@Multiply.Result
   )
   > &RigUnit_ModifyTransforms(
    ItemToModify=((Item=(Type=Bone,Name="dyn_hair_04_base_r"),Transform=(Rotation=(X=-0.0513,Y=-0.2129,Z=-0.3664,W=0.9043),Translation=(X=-1,Y=2.2,Z=0),Scale3D=(X=1,Y=1,Z=1))))
    WeightMinimum=0
    Weight=@Multiply.Result
   )
   > &RigUnit_ModifyTransforms(
    ItemToModify=((Item=(Type=Bone,Name="dyn_hair_base_back"),Transform=(Rotation=(X=-0.1274,Y=-0.2146,Z=-0.0283,W=0.9679),Translation=(X=0,Y=0,Z=0),Scale3D=(X=1,Y=1,Z=1))))
    WeightMinimum=0
    Weight=&RigVMFunction_MathDoubleMul(
     A=@Remap.Result
     B=&RigVMFunction_MathDoubleRemap(SourceMaximum=-15, SourceMinimum=0, TargetMaximum=1, TargetMinimum=0, Value=@"Get Delta Ref".Result.Rotation.Pitch).Result
    ).Result
   )
  ]}
]
hierarchy [
 root {loc=<0 0 0>, rot=<0 0 0 1>}
 pelvis {parent=root, loc=<0 -0.5301 98.0571>, rot=<0.0127 0.707 -0.0127 -0.707>}
 spine_01 {parent=pelvis, loc=<6.1208 -0.9328 0>, rot=<0 0 0.034 -0.9994>}
 spine_02 {parent=spine_01, loc=<6.5996 -0.2095 0>, rot=<0 0 0.0238 -0.9997>}
 spine_03 {parent=spine_02, loc=<6.4914 -0.1237 0>, rot=<0 0 0.0149 -0.9999>}
 spine_04 {parent=spine_03, loc=<6.9598 0.0794 0>, rot=<0 0 -0.0132 -0.9999>}
 spine_05 {parent=spine_04, loc=<6.9987 0 0>, rot=<0 0 -0.0015 -1>}
 neck_01 {parent=spine_05, loc=<11.3138 0.1529 0>, rot=<0 0 -0.2003 -0.9797>}
 neck_02 {parent=neck_01, loc=<4.5627 0 0>, rot=<0 0 0.0504 -0.9987>}
 head {parent=neck_02, loc=<5.3914 0 0>, rot=<0 0 0.115 -0.9934>}
 dyn_earring_r {parent=head, loc=<1.7745 -0.2606 6.4407>, rot=<0.0071 -0.7003 0.7138 -0.0008>}
 dyn_earring_l {parent=head, loc=<1.775 -0.2606 -6.4408>, rot=<0.0087 -0.7118 -0.7023 0.0022>}
 dyn_hair_01_01_r {parent=head, loc=<-0.3569 4.6891 8.2648>, rot=<-0.0921 -0.1172 0.9888 -0.0103>}
 dyn_hair_01_02_r {parent=dyn_hair_01_01_r, loc=<6.2766 -0.4921 -0.518>, rot=<0.0007 -0.0061 0.0605 -0.9981>}
 dyn_hair_01_03_r {parent=dyn_hair_01_02_r, loc=<6.2291 -1.0474 1.1017>, rot=<-0.0046 0.0771 0.1753 -0.9815>}
 dyn_hair_01_01_l {parent=head, loc=<0.0735 5.6825 -6.5035>, rot=<0.0081 0.3728 0.9262 -0.0553>}
 dyn_hair_01_02_l {parent=dyn_hair_01_01_l, loc=<6.2963 -1.0733 -0.2435>, rot=<0.0084 -0.0154 0.1795 -0.9836>}
 dyn_hair_01_03_l {parent=dyn_hair_01_02_l, loc=<6.1528 -0.4334 0.1138>, rot=<0.1526 0.1236 0.0659 -0.9783>}
 dyn_hair_03_base_r {parent=head, loc=<2.1986 -0.2081 7.873>, rot=<0.1349 -0.6958 0.7046 -0.0358>}
 dyn_hair_03_01_r {parent=dyn_hair_03_base_r, loc=<3.6278 0 0>, rot=<0 0 0 1>}
 dyn_hair_03_base_l {parent=head, loc=<3.4507 -0.4757 -7.2741>, rot=<-0.0019 -0.7907 -0.6122 0.0101>}
 dyn_hair_03_01_l {parent=dyn_hair_03_base_l, loc=<3.8201 -0.9515 0.3295>, rot=<0.0053 -0.0122 0.1387 -0.9902>}
 dyn_hair_base_back {parent=head, loc=<2.1906 -7.3597 0.1918>, rot=<0.0646 -0.9979 -0.0056 -0.0022>}
 dyn_hair_04_base_r {parent=dyn_hair_base_back, loc=<-2.0786 4.7025 -6.001>, rot=<-0.409 -0.0109 0.1458 -0.9007>}
 dyn_hair_04_01_r {parent=dyn_hair_04_base_r, loc=<0 0 0>, rot=<0 0 0 1>}
 dyn_hair_04_02_r {parent=dyn_hair_04_01_r, loc=<6.9164 0.1557 0.4426>, rot=<0.2258 0.0393 0.1218 -0.9657>}
 dyn_hair_04_03_r {parent=dyn_hair_04_02_r, loc=<6.1794 -0.5628 0.1388>, rot=<0.0977 0.1737 -0.0697 -0.9775>}
 dyn_hair_04_base_l {parent=dyn_hair_base_back, loc=<-0.5277 1.992 5.0267>, rot=<0.3603 0.1305 -0.0281 -0.9232>}
 dyn_hair_04_01_l {parent=dyn_hair_04_base_l, loc=<0 0 0>, rot=<0 0 0 1>}
 dyn_hair_04_02_l {parent=dyn_hair_04_01_l, loc=<7.9141 -2.9643 0.1632>, rot=<-0.1222 -0.3105 0.1901 -0.9233>}
 dyn_hair_04_03_l {parent=dyn_hair_04_02_l, loc=<5.3763 0.2298 -1.0028>, rot=<-0.0747 0.0304 -0.0716 -0.9942>}
 dyn_hair_05_base_l {parent=dyn_hair_base_back, loc=<4.3324 -0.8815 3.3301>, rot=<0.0277 -0.0201 0.0755 -0.9966>}
 dyn_hair_05_01_l {parent=dyn_hair_05_base_l, loc=<0 0 0>, rot=<0 0 0 1>}
 dyn_hair_05_02_l {parent=dyn_hair_05_01_l, loc=<8.6574 -0.1735 -2.0411>, rot=<0.0214 0.0626 0.055 -0.9963>}
 dyn_hair_05_base_r {parent=dyn_hair_base_back, loc=<3.6551 -1.2215 -0.4628>, rot=<-0.0278 -0.042 0.119 -0.9916>}
 dyn_hair_05_01_r {parent=dyn_hair_05_base_r, loc=<0 0 0>, rot=<0 0 0 1>}
 dyn_hair_05_02_r {parent=dyn_hair_05_01_r, loc=<8.7059 -0.1532 -1.2127>, rot=<0.0026 0.0723 -0.0527 -0.996>}
 dyn_hair_back_MSTR {parent=dyn_hair_base_back, loc=<4.0422 -0.9551 1.3409>, rot=<0.0277 -0.0201 0.0755 -0.9966>}
 dyn_hair_back_01 {parent=dyn_hair_back_MSTR, loc=<0 0 0>, rot=<0 0 0 1>}
 dyn_hair_back_02 {parent=dyn_hair_back_01, loc=<8.6188 -0.3328 -1.8758>, rot=<-0.0469 0.03 0.0491 -0.9972>}
 deform_hair_root_l {parent=head, loc=<6.6035 -4.1659 -4.5699>, rot=<0.0036 -0.866 -0.5 0.0021>}
 deform_hair_root_r {parent=head, loc=<4.8106 -3.0185 6.2331>, rot=<0.0034 -0.831 0.5562 -0.0023>}
 deform_cap {parent=head, loc=<11.6793 -1.4403 1.0386>, rot=<-0.069 0.1343 0.9726 0.1768>}
 dyn_hair_02_base_l {parent=head, loc=<3.0308 2.4576 -7.0214>, rot=<-0.003 0.6707 0.7415 -0.0163>}
 dyn_hair_02_01_l {parent=dyn_hair_02_base_l, loc=<4.4608 -0.4607 0.5252>, rot=<0.0716 -0.0018 0.1193 -0.9903>}
 dyn_hair_02_base_r {parent=head, loc=<1.6233 3.0185 7.1301>, rot=<0.2029 -0.7274 0.6549 -0.0285>}
 dyn_hair_02_01_r {parent=dyn_hair_02_base_r, loc=<3.8475 0.3922 -0.2332>, rot=<-0.1945 -0.2019 -0.141 -0.9495>}
 clavicle_l {parent=spine_05, loc=<5.4831 -0.1053 -3.2946>, rot=<0.0465 -0.7289 -0.009 -0.6829>}
 upperarm_l {parent=clavicle_l, loc=<9.8658 -1.2136 -0.4105>, rot=<0.0164 -0.3245 -0.0269 -0.9454>}
 lowerarm_l {parent=upperarm_l, loc=<26.465 -0.8345 0.259>, rot=<0.0036 0.0161 -0.2298 -0.9731>}
 upperarm_twist_01_l {parent=upperarm_l, loc=<7.2135 -0.1134 0>, rot=<0 0 0 1>}
 clavicle_r {parent=spine_05, loc=<5.4831 -0.1053 3.2946>, rot=<0.7289 0.0465 -0.6829 0.009>}
 upperarm_r {parent=clavicle_r, loc=<-9.8658 1.2136 0.4105>, rot=<0.0164 -0.3245 -0.0269 -0.9454>}
 lowerarm_r {parent=upperarm_r, loc=<-26.465 0.8345 -0.259>, rot=<0.0036 0.0161 -0.2298 -0.9731>}
 upperarm_twist_01_r {parent=upperarm_r, loc=<-7.2135 0.1134 0>, rot=<0 0 0 -1>}
]
