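# Physics asset for the RenegadeWhip female character's FaceAcc part, set up for dynamic hair physics.
# Capsules cover head, upper spine, clavicles and upper arms as body shapes, joined by angularly free constraints.
# Angle-limited hair chains form two three-segment strands from the head, plus left, right and back chains rooted through a shared back base.
# Extra angularly free constraints cross-link the back chain to both side chains.
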PhysicsAsset /BRCosmetics/Characters/Player/Female/Medium/Bodies/F_MED_RenegadeWhip/Meshes/Parts/F_MED_RenegadeWhip_FaceAcc_Physics
bodies {
 clavicle_l [{
   shapeType=Capsule
   length=14.2
   radius=7.605
   shapeName=clavicle_r_capsule
   center=<10.6063 2.1103 -3.082>
   rotation=<rot -71.2713 -178.2649 13.6972>
  }]
 clavicle_r [{
   shapeType=Capsule
   length=10.1
   radius=7.605
   shapeName=clavicle_r_capsule
   center=<-9.127 -1.6006 1.98>
   rotation=<rot 71.044 1.7286 -16.5051>
  }]
 dyn_hair_01_01_l [{
   shapeType=Capsule
   length=4.235
   radius=1.7614
   shapeName=dyn_hair_01_01_l_capsule
   center=<3.0332 -0.1882 0.0309>
   rotation=<rot -1.6385 77.5201 81.3416>
  }]
 dyn_hair_01_01_r [{
   shapeType=Capsule
   length=3.8766
   radius=1.5532
   shapeName=dyn_hair_01_01_r_capsule
   center=<3.266 -0.0133 -0.0925>
   rotation=<rot -0.0031 87.3003 85.5617>
  }]
 dyn_hair_01_02_l [{
   shapeType=Capsule
   length=4.1535
   radius=1.5396
   shapeName=dyn_hair_01_02_l_capsule
   center=<2.7427 -0.1821 -0.1259>
   rotation=<rot 0.1264 86.6979 85.7612>
  }]
 dyn_hair_01_02_r [{
   shapeType=Capsule
   length=3.7385
   radius=1.8046
   shapeName=dyn_hair_01_02_r_capsule
   center=<3.5854 -0.2427 0.4798>
   rotation=<rot 0 81.3173 -78.2937>
  }]
 dyn_hair_01_03_l [{
   shapeType=Capsule
   length=4.2641
   radius=1.8344
   shapeName=dyn_hair_01_03_l_capsule
   center=<3.0674 -0.1795 0.0328>
   rotation=<rot 0 83.2541 -89.757>
  }]
 dyn_hair_01_03_r [{
   shapeType=Capsule
   length=3.005
   radius=2.0826
   shapeName=dyn_hair_01_03_r_capsule
   center=<2.974 -0.4226 0.0001>
   rotation=<rot -0.0031 76.3464 88.2417>
  }]
 dyn_hair_04_01_l [{
   shapeType=Capsule
   length=3.9936
   radius=2.6138
   shapeName=dyn_hair_04_01_l_capsule
   center=<3.5341 -1.7443 0.7978>
   rotation=<rot 85.6141 5.0014 22.9968>
  }]
 dyn_hair_04_01_r [{
   shapeType=Capsule
   length=4.1627
   radius=2.3934
   shapeName=dyn_hair_04_01_r_capsule
   center=<3.6388 -0.496 0.3085>
   rotation=<rot 0.0045 83.3907 -90.0341>
  }]
 dyn_hair_04_02_l [{
   shapeType=Capsule
   length=1.8224
   radius=2.7243
   shapeName=dyn_hair_04_02_l_capsule
   center=<2.3167 -0.0743 -0.1603>
   rotation=<rot 23.8195 87.497 79.592>
  }]
 dyn_hair_04_02_r [{
   shapeType=Capsule
   length=2.0062
   radius=2.429
   shapeName=dyn_hair_04_02_r_capsule
   center=<3.4866 -0.0271 0.051>
   rotation=<rot -1.6065 94.2961 -89.1423>
  }]
 dyn_hair_04_03_l [{
   shapeType=Capsule
   length=6.9479
   radius=2.4684
   shapeName=dyn_hair_04_03_l_capsule
   center=<4.1254 -0.111 0.6093>
   rotation=<rot 0 87.061 -82.375>
  }]
 dyn_hair_04_03_r [{
   shapeType=Capsule
   length=6.6062
   radius=1.729
   shapeName=dyn_hair_04_02_r_capsule
   center=<4.5495 -0.4843 0.3321>
   rotation=<rot -1.7239 96.0021 -86.0843>
  }]
 dyn_hair_04_base_l [{shapeType=Sphere, radius=0.5, shapeName=dyn_hair_04_base_l_sphere}]
 dyn_hair_04_base_r [{shapeType=Sphere, radius=0.5, shapeName=dyn_hair_04_base_r_sphere}]
 dyn_hair_back_01 [
  {
   shapeType=Capsule
   length=2.5597
   radius=2.8247
   shapeName=dyn_hair_05_01_r_capsule
   center=<7.3477 -1.6836 -1.8014>
   rotation=<rot -0.0255 88.0262 -1.9169>
  }
  {
   shapeType=Capsule
   length=3.3
   radius=1.5
   shapeName=dyn_hair_05_01_r_capsule
   center=<4.5307 0.3725 -2.2634>
   rotation=<rot 73.0293 -179.9999 180>
  }
 ]
 dyn_hair_back_02 [{
   shapeType=Capsule
   length=3.7908
   radius=2.9439
   shapeName=dyn_hair_05_02_r_capsule
   center=<5.6095 -0.7979 0.6414>
   rotation=<rot -0.0294 87.8856 -167.4373>
  }]
 dyn_hair_back_MSTR [{shapeType=Capsule, length=0.505, radius=0.505, shapeName=dyn_hair_back_MSTR_capsule, rotation=<rot 0 -90 0>}]
 dyn_hair_base_back [{shapeType=Capsule, length=0.1, radius=0.5, shapeName=dyn_hair_base_back_capsule, center=<6.8952 0.1427 -0.2826>}]
 head [
  {
   shapeType=Capsule
   length=8.6114
   radius=5.9537
   shapeName=head_capsule
   center=<1.1353 1.0556 0.1993>
   rotation=<rot 0.1129 -81.5316 -89.8304>
  }
  {shapeType=Capsule, length=8.3216, radius=5.5, shapeName=head_capsule, center=<0.4336 7.3059 -0.0042>, rotation=<rot 89.445 -0.0001 0>}
 ]
 spine_05 [
  {shapeType=Capsule, length=39.3, radius=9.505, shapeName=spine_05_capsule, center=<3.6143 -0.4264 0>, rotation=<rot 0 -174.7973 0>}
  {shapeType=Capsule, length=17.799, radius=12.201, shapeName=spine_05_capsule, center=<-3.7801 -0.4969 0>, rotation=<rot 0 3.1918 0>}
 ]
 upperarm_l [{
   shapeType=Capsule
   length=23.1
   radius=5.005
   shapeName=upperarm_l_capsule
   center=<17.1429 0.185 -0.2298>
   rotation=<rot -89.586 -137.2138 137.2241>
  }]
 upperarm_r [{
   shapeType=Capsule
   length=22.9685
   radius=5.1365
   shapeName=upperarm_l_capsule
   center=<-18.4655 0.1792 -0.0409>
   rotation=<rot 89.5785 43.9 -136.1165>
  }]
}
constraints [
 {bone1Name=dyn_hair_04_02_r, bone2Name=dyn_hair_04_01_r, swing1LimitDegrees=40, swing2LimitDegrees=30, twistLimitDegrees=10}
 {bone1Name=dyn_hair_04_01_r, bone2Name=dyn_hair_04_base_r, swing1LimitDegrees=18, swing2LimitDegrees=5, twistLimitDegrees=5}
 {bone1Name=dyn_hair_04_03_l, bone2Name=dyn_hair_04_02_l, swing1LimitDegrees=30, swing2LimitDegrees=15, twistLimitDegrees=10}
 {bone1Name=dyn_hair_04_02_l, bone2Name=dyn_hair_04_01_l, swing1LimitDegrees=30, swing2LimitDegrees=20, twistLimitDegrees=10}
 {bone1Name=dyn_hair_01_03_r, bone2Name=dyn_hair_01_02_r, swing1LimitDegrees=60, swing2LimitDegrees=60, twistLimitDegrees=15}
 {bone1Name=dyn_hair_01_02_r, bone2Name=dyn_hair_01_01_r, swing1LimitDegrees=60, swing2LimitDegrees=60, twistLimitDegrees=15}
 {bone1Name=dyn_hair_01_01_r, bone2Name=head, swing1LimitDegrees=60, swing2LimitDegrees=60, twistLimitDegrees=15}
 {bone1Name=dyn_hair_01_03_l, bone2Name=dyn_hair_01_02_l, swing1LimitDegrees=60, swing2LimitDegrees=60, twistLimitDegrees=15}
 {bone1Name=dyn_hair_01_02_l, bone2Name=dyn_hair_01_01_l, swing1LimitDegrees=60, swing2LimitDegrees=60, twistLimitDegrees=20}
 {bone1Name=dyn_hair_01_01_l, bone2Name=head, swing1LimitDegrees=45, swing2LimitDegrees=45, twistLimitDegrees=20}
 {
  bone1Name=head
  bone2Name=spine_05
  swing1LimitDegrees=45
  swing1Motion=Free
  swing2LimitDegrees=45
  swing2Motion=Free
  twistLimitDegrees=45
  twistMotion=Free
 }
 {
  bone1Name=clavicle_l
  bone2Name=spine_05
  swing1LimitDegrees=45
  swing1Motion=Free
  swing2LimitDegrees=45
  swing2Motion=Free
  twistLimitDegrees=45
  twistMotion=Free
 }
 {
  bone1Name=upperarm_l
  bone2Name=clavicle_l
  swing1LimitDegrees=45
  swing1Motion=Free
  swing2LimitDegrees=45
  swing2Motion=Free
  twistLimitDegrees=45
  twistMotion=Free
 }
 {
  bone1Name=clavicle_r
  bone2Name=spine_05
  swing1LimitDegrees=45
  swing1Motion=Free
  swing2LimitDegrees=45
  swing2Motion=Free
  twistLimitDegrees=45
  twistMotion=Free
 }
 {
  bone1Name=upperarm_r
  bone2Name=clavicle_r
  swing1LimitDegrees=45
  swing1Motion=Free
  swing2LimitDegrees=45
  swing2Motion=Free
  twistLimitDegrees=45
  twistMotion=Free
 }
 {
  bone1Name=dyn_hair_base_back
  bone2Name=head
  swing1LimitDegrees=45
  swing1Motion=Free
  swing2LimitDegrees=45
  swing2Motion=Free
  twistLimitDegrees=45
  twistMotion=Free
 }
 {bone1Name=dyn_hair_04_03_r, bone2Name=dyn_hair_04_02_r, swing1LimitDegrees=25, swing2LimitDegrees=30, twistLimitDegrees=10}
 {bone1Name=dyn_hair_04_01_l, bone2Name=dyn_hair_04_base_l, swing1LimitDegrees=50, swing2LimitDegrees=50, twistLimitDegrees=10}
 {
  bone1Name=dyn_hair_04_base_l
  bone2Name=dyn_hair_base_back
  swing1LimitDegrees=45
  swing1Motion=Free
  swing2LimitDegrees=45
  swing2Motion=Free
  twistLimitDegrees=45
  twistMotion=Free
 }
 {
  bone1Name=dyn_hair_04_base_r
  bone2Name=dyn_hair_base_back
  swing1LimitDegrees=45
  swing1Motion=Free
  swing2LimitDegrees=45
  swing2Motion=Free
  twistLimitDegrees=45
  twistMotion=Free
 }
 {
  bone1Name=dyn_hair_back_MSTR
  bone2Name=dyn_hair_base_back
  swing1LimitDegrees=45
  swing1Motion=Free
  swing2LimitDegrees=45
  swing2Motion=Free
  twistLimitDegrees=45
  twistMotion=Free
 }
 {bone1Name=dyn_hair_back_01, bone2Name=dyn_hair_back_MSTR, swing1LimitDegrees=35, swing2LimitDegrees=25, twistLimitDegrees=10}
 {bone1Name=dyn_hair_back_02, bone2Name=dyn_hair_back_01, swing1LimitDegrees=35, swing2LimitDegrees=20, twistLimitDegrees=10}
 {
  bone1Name=dyn_hair_back_02
  bone2Name=dyn_hair_04_03_r
  swing1LimitDegrees=45
  swing1Motion=Free
  swing2LimitDegrees=45
  swing2Motion=Free
  twistLimitDegrees=45
  twistMotion=Free
 }
 {
  bone1Name=dyn_hair_back_02
  bone2Name=dyn_hair_04_03_l
  swing1LimitDegrees=45
  swing1Motion=Free
  swing2LimitDegrees=45
  swing2Motion=Free
  twistLimitDegrees=45
  twistMotion=Free
 }
 {
  bone1Name=dyn_hair_04_01_l
  bone2Name=dyn_hair_back_01
  swing1LimitDegrees=45
  swing1Motion=Free
  swing2LimitDegrees=45
  swing2Motion=Free
  twistLimitDegrees=45
  twistMotion=Free
 }
]
